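# Test Niagara system for raymarched lava bubbles, apparently drawn as metaballs rising from a ground surface.
# Metaball001 spawns rising particles, packing position, size, shrinking capsule, orbiting second-blob and ground data into Vector4s.
# It reads five of its own particles back to emitter level, checks whether the camera is inside a box, and renders through mesh and sprite renderers.
# Minimal spawns ten small sprites per new metaball, tracking their parent through Particle Attribute Reader.
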NiagaraSystem /Game/Effects/Fort_Effects/Effects/RaymarchingBubble/LavaTest_RM
dependencies {
 usedDataInterfaces [
  /Script/Niagara.NiagaraDataInterfaceSimpleCounter
  &NiagaraDataInterfaceParticleRead
  &NiagaraDataInterfaceCurve
  &NiagaraDataInterfaceCamera
  &NiagaraDataInterfaceCollisionQuery
 ]
 usedDynamicInputs []
 usedModules [&SystemState]
}
emitters {
 Metaball001 {
  emitterSpawnScript [/SetVariables(Emitter.ParticleAttribute=DataInterface() {propertyValues {EmitterBinding {bindingMode=Other, emitterName=Metaball001}}}) {isSetParameters=true}]
  emitterUpdateScript [
   &EmitterState(Life Cycle Mode=::NewEnumerator1, Loop Behavior=::NewEnumerator1, Loop Duration Mode=::NewEnumerator1, UseLoopDelay=false)
   &SpawnRate(Use Spawn Probability=false)
   /SetVariables(Emitter.Position:NiagaraPosition=$Engine.Emitter.SimulationPosition, Emitter.Scale:Vector3f=<1 1 1>) {isSetParameters=true}
   &self:ReadMySelf(Particle Attribute Reader:NiagaraDataInterfaceParticleRead=$Emitter.ParticleAttribute) {module=ReadMySelf}
   &self:ReadMySelf(Particle Attribute Reader:NiagaraDataInterfaceParticleRead=$Emitter.ParticleAttribute, Particle Index=1) {module=ReadMySelf001}
   &self:ReadMySelf(Particle Attribute Reader:NiagaraDataInterfaceParticleRead=$Emitter.ParticleAttribute, Particle Index=2) {module=ReadMySelf002}
   &self:ReadMySelf(Particle Attribute Reader:NiagaraDataInterfaceParticleRead=$Emitter.ParticleAttribute, Particle Index=3) {module=ReadMySelf003}
   &self:ReadMySelf(Particle Attribute Reader:NiagaraDataInterfaceParticleRead=$Emitter.ParticleAttribute, Particle Index=4) {module=ReadMySelf004}
   /SetVariables(Emitter.Bound:Vector3f=&Divide_Vector(A:Vector3f=<256 256 256>, B:Vector3f=<100 100 100>)) {isSetParameters=true}
   &self:IsCameraInsideBox(Bound:Vector3f=<300 300 300>, Camera Query=DataInterface() {propertyValues {PlayerControllerIndex=0, bRequireCurrentFrameData=false}}) {module=IsCameraInsideBox}
  ]
  particleSpawnScript [
   &InitializeParticle(
    Color Mode=::NewEnumerator0
    Lifetime=5
    Sprite Size Mode=::NewEnumerator3
    Uniform Sprite Size=&RandomRangeFloat(Minimum=12, Maximum=15)
    UsePositionOffset=false
    Write Material Random=true
   )
   &ShapeLocation(
    Cylinder Height=1
    MaskSpawnGroup=false
    Offset=<0 0 -128>
    Offset Mode=::NewEnumerator0
    Override Seed=false
    Shape Primitive=::NewEnumerator1
    Surface Only=false
    UseOwnerScale=true
   )
   &AddVelocity(Velocity=&RandomRangeVector(Minimum=<0 0 45.0256>, Maximum=<1 1 73.5273>))
   &self:TraceStartPos(DownOffset=128) {module=TraceStartPos}
  ]
  particleUpdateScript [
   &ParticleState()
   &Timeline(
    Initial Start Time:NiagaraFloat=&RandomRangeFloat(Minimum=10, Maximum=100)
    Play Rate=&FloatFromCurve(FloatCurve=DataInterface() {propertyValues {Curve {keys [<at 0.3968 0.4721>,<at 0.4686 1>]}}})
    Timeline End=999999
   )
   &ScaleSpriteSize(Uniform Curve Sprite Scale=DataInterface() {propertyValues {Curve {
       defaultValue=1
       keys [
        {arriveTangent=9.0725, interpMode=RCIM_Cubic, leaveTangent=9.0725, tangentMode=RCTM_User, time=0, value=0}
        {arriveTangent=0.5491, interpMode=RCIM_Cubic, leaveTangent=0.5491, tangentMode=RCTM_User, time=0.13, value=0.9847}
        {arriveTangent=0.0679, interpMode=RCIM_Cubic, leaveTangent=0.0679, tangentMode=RCTM_User, time=0.6857, value=0.9653}
        {arriveTangent=-3.2218, interpMode=RCIM_Cubic, leaveTangent=-3.2218, tangentMode=RCTM_User, time=1, value=0}
       ]
      }}})
   &Drag(Drag=0.3, Use Linear Drag=true, Use Rotational Drag=true)
   &SolveForcesAndVelocity(Clamp Velocity=true, Limit Acceleration=false, Speed Limit=98.2765)
   /SetVariables(
    Particles.CapHeight:NiagaraFloat=&FloatFromCurve(
     FloatCurve=DataInterface() {propertyValues {Curve {keys [
         {interpMode=RCIM_Cubic, tangentMode=RCTM_Break, time=0, value=1}
         {arriveTangent=0.2617, interpMode=RCIM_Cubic, leaveTangent=0.2617, tangentMode=RCTM_User, time=0.3, value=1}
         {arriveTangent=-1.5011, interpMode=RCIM_Cubic, tangentMode=RCTM_Break, time=0.55, value=0.01}
        ]}}}
     CurveIndex=&Multiply_Float(A=$Particles.NormalizedAge, B=1)
     Scale Curve=100
    )
    Particles.CapRadius:NiagaraFloat=&FloatFromCurve(
     FloatCurve=DataInterface() {propertyValues {Curve {keys [
         {interpMode=RCIM_Cubic, leaveTangent=-0.2784, tangentMode=RCTM_Break, time=0.4, value=1}
         {arriveTangent=-2.0114, tangentMode=RCTM_User, time=0.55, value=0}
        ]}}}
     CurveIndex=&Multiply_Float(A=$Particles.NormalizedAge, B=1)
     Scale Curve=&Multiply_Float(A=&MakeFloatFromVector2d(Vector 2D=$Particles.SpriteSize), B=1)
    )
    Particles.CapScaleTip:NiagaraFloat=&FloatFromCurve(
     FloatCurve=DataInterface() {propertyValues {Curve {keys [{interpMode=RCIM_Cubic, tangentMode=RCTM_Break, time=0.1, value=1}, <at 0.5 0>]}}}
     CurveIndex=&Multiply_Float(A=$Particles.NormalizedAge, B=1)
    )
   ) {isSetParameters=true}
   /SetVariables(Particles.PositionAndSizeInfo:Vector4f=&MakeVector4FromVectorAndFloat(
     "Vector (RGB)"=&TransformPosition(Position=$Particles.Position, Source Space=::Simulation, Destination Space=::World, Apply Scale=true)
     "Float (Alpha)"=&MakeFloatFromVector2d(Vector 2D=$Particles.SpriteSize)
    )) {isSetParameters=true}
   /SetVariables(Particles.CapsuleInfo:Vector4f=&MakeVector4(X=$Particles.CapHeight, Y=$Particles.CapRadius, Z=$Particles.CapScaleTip, W=$Particles.CapsuleOffset)) {isSetParameters=true}
   &self:SecondParticleMotion(
    Age:NiagaraFloat=$Particles.Timeline.Time
    Offset:NiagaraFloat=&Multiply_Float(
     A=&RandomRangeFloat(Minimum=0.5, Maximum=0.8)
     B=&MakeFloatFromVector2d(Vector 2D=$Particles.SpriteSize)
    )
    RandomZ:Vector3f=&NormalizeVector(
     Vector To Normalize=&RandomRangeVector(Minimum=<-1 -1 -1>, Maximum=<1 1 1>)
     Safe Normalize=true
     Safe Fallback Threshold=0
     Safe Fallback Vector=<0 0 0.001>
    )
    Speed:NiagaraFloat=&RandomRangeFloat(Minimum=100, Maximum=180)
   ) {module=SecondParticleMotion}
   /SetVariables(Particles.SecondBlobInfo:Vector4f=&MakeVector4(
     X=&MakeFloatFromVector(
      Vector=&TransformPosition(Position=$Particles.SecondParticleMotion.SecBPosition, Source Space=::Simulation, Destination Space=::World)
      Channel=::NewEnumerator0
     )
     Y=&MakeFloatFromVector(
      Vector=&TransformPosition(Position=$Particles.SecondParticleMotion.SecBPosition, Source Space=::Simulation, Destination Space=::World)
      Channel=::NewEnumerator1
     )
     Z=&MakeFloatFromVector(
      Vector=&TransformPosition(Position=$Particles.SecondParticleMotion.SecBPosition, Source Space=::Simulation, Destination Space=::World)
      Channel=::NewEnumerator2
     )
     W=&Multiply_Float(
      A=&MakeFloatFromVector2d(Vector 2D=$Particles.SpriteSize)
      B=&RandomRangeFloat(Minimum=0.9, Maximum=1.1)
     )
    )) {isSetParameters=true}
   /SetVariables(Particles.GroundInfo:Vector4f=&MakeVector4(
     X=0
     Y=0
     Z=1
     W=&MakeFloatFromVector(Vector=Linked() {linkedVariable {name=Particles.StartPos, type=NiagaraPosition}}, Channel=::NewEnumerator2)
    )) {isSetParameters=true}
  ]
  rendererClasses [&NiagaraMeshRendererProperties, &NiagaraSpriteRendererProperties]
  renderers [
   {rendererClass=&NiagaraMeshRendererProperties, rendererIndex=0}
   {rendererClass=&NiagaraMeshRendererProperties, rendererIndex=1}
   {rendererClass=&NiagaraSpriteRendererProperties, rendererIndex=2}
  ]
 }
 Minimal {
  emitterSpawnScript [
   /SetVariables() {isSetParameters=true}
   /SetVariables(Emitter.Particle Attribute Reader=DataInterface() {propertyValues {EmitterBinding {bindingMode=Other, emitterName=Metaball001}}}) {isSetParameters=true}
  ]
  emitterUpdateScript [
   &EmitterState(Life Cycle Mode=::NewEnumerator1, Loop Behavior=::NewEnumerator1, Loop Duration Mode=::NewEnumerator1, UseLoopDelay=false)
   &self:SpawnFromMather(Particle Attribute Reader:NiagaraDataInterfaceParticleRead=$Emitter.Particle Attribute Reader) {module=SpawnFromMather}
   &SpawnPerFrame(Spawn Count:NiagaraInt32=&Multiply_Int(A=$Emitter.ThisFrameSpawned, B=10), Use Spawn Probability=false)
  ]
  particleSpawnScript [
   &PartitionParticles(Number Of Partitions=$Emitter.ThisFrameSpawned)
   &InitializeParticle(
    Color Mode=::NewEnumerator0
    Lifetime=&RandomRangeFloat(Minimum=1, Maximum=5)
    Sprite Size Mode=::NewEnumerator3
    Uniform Sprite Size=&RandomRangeFloat(Minimum=1, Maximum=5)
    UsePositionOffset=false
    Write Material Random=true
   )
   &self:ScratchModule_01(
    PartID:NiagaraInt32=$Particles.PartitionParticles.Partition
    Particle Attribute Reader:NiagaraDataInterfaceParticleRead=$Emitter.Particle Attribute Reader
   ) {module=ScratchModule_01}
   &ShapeLocation(MaskSpawnGroup=false, Offset=<0 0 -20>, Offset Mode=::NewEnumerator0, Override Seed=false, Sphere Radius=10, UseOwnerScale=true)
  ]
  particleUpdateScript [
   &ParticleState()
   &ScaleSpriteSize(Uniform Curve Sprite Scale=DataInterface() {propertyValues {Curve {
       defaultValue=0
       keys [
        {arriveTangent=9.0725, interpMode=RCIM_Cubic, leaveTangent=9.0725, tangentMode=RCTM_User, time=0, value=0}
        {arriveTangent=-0.06, interpMode=RCIM_Cubic, leaveTangent=-0.06, tangentMode=RCTM_User, time=0.1408, value=0.9847}
        {arriveTangent=-1.146, interpMode=RCIM_Cubic, leaveTangent=-1.146, time=0.7505, value=0.8179}
        {arriveTangent=-2.6043, interpMode=RCIM_Cubic, leaveTangent=-2.6043, tangentMode=RCTM_User, time=1, value=0}
       ]
      }}})
   &self:ScratchModule_02(
    Niagara ID:NiagaraID=$Particles.SpawnID
    Offet:Vector3f=&AddVectorToPosition(
     Position=$Particles.ShapeLocation.ShapePosition
     Vector=&MakeVector(
      X=0
      Y=0
      Z=&FloatFromCurve(
       FloatCurve=DataInterface() {propertyValues {Curve {keys [{interpMode=RCIM_Cubic, tangentMode=RCTM_Break, time=0, value=0}, <at 1 1>]}}}
       Scale Curve=&RandomRangeFloat(Minimum=20, Maximum=30)
      )
     )
    )
    Particle Attribute Reader:NiagaraDataInterfaceParticleRead=$Emitter.Particle Attribute Reader
    SampleID:NiagaraInt32=$Particles.SampleID
   ) {module=ScratchModule_02}
  ]
  rendererClasses [&NiagaraSpriteRendererProperties]
  renderers [{rendererClass=&NiagaraSpriteRendererProperties}]
 }
}
summary {
 systemName=LavaTest_RM
 emitters [
  {emitterName=Minimal, rendererClasses [&NiagaraSpriteRendererProperties]}
  {emitterName=Metaball001, rendererClasses [&NiagaraMeshRendererProperties, &NiagaraSpriteRendererProperties]}
 ]
 userVariables []
}
