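# Translucent unlit material for Niagara mesh particles, a glossy blob that appears to be the raymarched bubble effect's fallback.
# A faked sky-atmosphere light direction, bent by view angle, drives diffuse and Fresnel terms into a curve-atlas color ramp.
# GGX highlight, Fresnel-masked skylight reflection and overlay tint are added, then scaled by particle color, with depth fade on alpha.
# A refraction branch samples scene color at per-channel offsets for chromatic dispersion.
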Material /Game/Effects/Fort_Effects/Effects/RaymarchingBubble/M_FallBackBlob
graph {
 properties {
  bAutomaticallySetUsageInEditor=false
  bUsedWithNiagaraMeshParticles=true
  blendMode=BLEND_Translucent
  shadingModel=MSM_Unlit
  usedShadingModels=Unlit
 }
 nodes [
  Constant()
  n1 = VectorParameter() {group="######ColorTweak######", parameterName=EnvReflectColor, defaultValue=<rgba 0.2 0.2 0.2 1>}
  n2 = CameraVectorWS()
  n3 = NamedRerouteUsage()
  n4 = OneMinus(None=Abs(None=DotProduct(A=@n2, B=@n3)))
  n5 = Add(A=@n4, B=Constant())
  n6 = Constant() {r=1}
  n7 = Power(Base=Saturate(None=Multiply(A=@n5, B=@n6)), Exp=Constant() {r=2.5})
  n8 = NamedRerouteUsage()
  n9 = Multiply(
   A=MaterialFunctionCall(Input=Saturate(None=@n7), Target Low=ScalarParameter() {defaultValue=0.02, group="######ColorTweak######", parameterName=EnvReflectCenterStrength}) {materialFunction=&RemapValueRange}.Result
   B=SkyLightEnvMapSample(Direction=ReflectionVectorWS(CustomWorldNormal=@n8), Roughness=Constant())
  )
  n10 = Constant4Vector() {constant=<rgba 0.2 0.2 0.2 1>}
  n11 = VectorParameter() {group="######ColorTweak######", parameterName="OverLayColor(SimTheSkyLight)", defaultValue=<rgba 0.09 0.1324 0.3 1>}
  n12 = Multiply(A=@n10.RGB, B=@n11.RGB) {constB=0.1}
  n13 = Add(
   A=NamedRerouteUsage()
   B=MaterialFunctionCall("Xb1,PS4"=@n12, "PC-High&&Epic,XSX,PS5"=@n12, "PC-Perf,Android"=@n12, Switch=@n12, "PC-Low(NotPerf)&&Medium"=@n12) {materialFunction=&PlatformSwitch_1D}.Result
  )
  n14 = DotProduct(
   A=SkyAtmosphereLightDirection()
   B=MaterialFunctionCall() {materialFunction=&CameraDirectionVector}.Result
  )
  n15 = Constant() {r=-1.0301}
  n16 = Saturate(None=Multiply(A=Add(A=@n14, B=@n15), B=Constant() {r=2}))
  n17 = Constant() {r=1}
  n18 = OneMinus(None=Saturate(None=Power(Base=@n16, Exp=@n17)))
  n19 = NamedRerouteUsage()
  n20 = SkyAtmosphereLightDirection()
  n21 = Constant3Vector() {constant=<rgba 0 0 1 0>}
  n22 = Constant() {r=0.6}
  n23 = CameraVectorWS()
  n24 = MaterialFunctionCall() {materialFunction=&CameraDirectionVector}
  n25 = MaterialFunctionCall!(Plane Normal=@n24.Result, Vector=Constant3Vector() {constant=<rgba 0 0 1 0>}) {materialFunction=/Engine/Functions/Engine_MaterialFunctions02/Math/ProjectVectorOntoPlane}
  n26 = Add(A=@n14, B=Constant() {r=1.4})
  n27 = Constant() {r=1}
  n28 = Power(Base=Saturate(None=Multiply(A=@n26, B=@n27)), Exp=Constant() {r=1})
  n29 = LinearInterpolate(
   A=LinearInterpolate(A=LinearInterpolate(A=@n20, B=@n21, Alpha=@n22), B=Multiply(A=@n23) {constB=-1}, Alpha=Constant() {r=0.4})
   B=LinearInterpolate(A=Normalize(VectorInput=@n25.Result), B=@n24.Result) {constAlpha=0.8}
   Alpha=OneMinus(None=Saturate(None=@n28))
  )
  n30 = DotProduct(A=@n19, B=@n29)
  n31 = Multiply(A=Add(A=Multiply(A=@n30) {constB=-1}, B=Constant() {r=0.25}), B=Constant() {r=1})
  n32 = Saturate(None=Power(Base=Saturate(None=@n31), Exp=Constant() {r=1.5}))
  n33 = Add(A=@n4, B=Constant() {r=0.2})
  n34 = Constant() {r=1.5}
  n35 = Power(Base=Saturate(None=Multiply(A=@n33, B=@n34)), Exp=Constant() {r=1.5})
  n36 = Add(A=Multiply(A=Multiply(A=@n18, B=@n32), B=Constant() {r=0.5}), B=Multiply(A=Saturate(None=@n35), B=Constant() {r=0.4}))
  n37 = TextureSampleParameter2D(UVs=MaterialFunctionCall!(
    CurveIndex=ScalarParameter() {group="######ColorTweak######", parameterName=ColorCurveIndex}
    CurveNum=Constant() {r=4}
    GreyMapIn=Saturate(None=@n36)
   ) {materialFunction=/Game/Effects/Niagara/MasterMaterials/Functions/Common/FXCommon_CurveAtlasUV}.UV) {
   group="######ColorTweak######"
   parameterName=CurveTexture
   samplerType=SAMPLERTYPE_LinearColor
   texture=&/CLA_LavaBlob
  }
  n38 = ScalarParameter() {defaultValue=0.02, group="######ColorTweak######", parameterName=HighLightPointSize}
  n39 = Add(
   A=@n37.RGB
   B=Multiply(
    A=MaterialFunctionCall!(LightDirctionIn=@n29, CameraVectorIn=@n2, Roughness=@n38, NormalIn=@n19) {materialFunction=/Game/Effects/Niagara/MasterMaterials/Functions/Common/FXCommon_GGX}.Result
    B=VectorParameter() {group="######ColorTweak######", parameterName=HighLightColor, defaultValue=<rgba 100 91 58 1>}.RGB
   )
  )
  n40 = NamedRerouteDeclaration(None=Add(A=Multiply(A=@n1.RGB, B=@n9), B=Add(A=@n13, B=@n39))) {name=RenderStyleOutput, nodeColor=<rgba 0.1953 0 1 1>}
  n41 = ParticleColor()
  MaterialFunctionCall(Emissive=Multiply(A=Divide(A=@n40) {constB=1.97}, B=@n41.RGB)) {materialFunction=&MF_FX_EmissiveAdjustments}
  Multiply(A=DepthFade(), B=@n41.A)
  NamedRerouteDeclaration(None=PixelNormalWS()) {name=Normal, nodeColor=<rgba 0.5938 1 0 1>}
  n42 = Subtract(A=NamedRerouteUsage(), B=CameraPositionWS())
  n43 = MaterialFunctionCall!(
   Ray Direction=MaterialFunctionCall() {materialFunction=&CameraDirectionVector}.Result
   Surface Normal=NamedRerouteUsage()
   Refractive Index Origin=Constant() {r=1}
   Refractive Index Target=Constant() {r=1.05}
  ) {materialFunction=/Engine/Functions/Engine_MaterialFunctions01/Vectors/Refract}
  n44 = Constant4Vector() {constant=<rgba 0.0105 0.01 0.0095 1>}
  n45 = Add(A=@n42, B=Divide(A=@n43.Vector, B=@n44.R))
  n46 = ComponentMask(None=SceneColor!(UVs=MaterialFunctionCall!(Camera Relative Position=@n45, X Camera to World Vector, Y Camera to World Vector, Z Camera to World Vector, "Tan(FOV/2)*[1,Screen Res Y/X]") {materialFunction=/Engine/Functions/Engine_MaterialFunctions02/Math/ManualWorldToScreenUVsTransform}.Screen UVs).RGB) {r=true}
  n47 = Add(A=@n42, B=Divide(A=@n43.Vector, B=@n44.G))
  n48 = ComponentMask(None=SceneColor(UVs=MaterialFunctionCall(Camera Relative Position=@n47) {materialFunction=&ManualWorldToScreenUVsTransform}.Screen UVs).RGB) {g=true}
  n49 = Add(A=@n42, B=Divide(A=@n43.Vector, B=@n44.B))
  n50 = ComponentMask(None=SceneColor(UVs=MaterialFunctionCall(Camera Relative Position=@n49) {materialFunction=&ManualWorldToScreenUVsTransform}.Screen UVs).RGB) {b=true}
  Reroute(None=Multiply(
    A=MaterialFunctionCall(X=@n46, Y=@n48, Z=@n50) {materialFunction=&MakeFloat3}.Result
    B=@n10.RGB
   ))
 ]
}
