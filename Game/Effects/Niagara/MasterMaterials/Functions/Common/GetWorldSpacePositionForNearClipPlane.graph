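# Material function, exposed to the library, that returns the world-space position of the current pixel projected onto the camera's near clip plane.
# It remaps viewport UV to centered screen coordinates with Y flipped, scales them by the near plane distance and builds a float4 with that distance as depth.
# A Custom node multiplies this by ResolvedView.ScreenToWorld, converted with LWCToFloat, and the XYZ result is output.
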MaterialFunction /Game/Effects/Niagara/MasterMaterials/Functions/Common/GetWorldSpacePositionForNearClipPlane
graph {
 expressions {
  MaterialExpressionComponentMask_4 {class=MaterialExpressionComponentMask, inputs [None {from=MaterialExpressionCustom_6}], properties {b=true, g=true, r=true}}
  MaterialExpressionConvert_0 {
   class=MaterialExpressionConvert
   inputs [None {from=MaterialExpressionScreenPosition_2, output=ViewportUV}]
   properties {
    nodeName=Break Float2
    convertInputs [{
      type=Vector2
      expressionInput {
       expression=&self:MaterialExpressionScreenPosition_2
       inputName=None
       mask=0
       maskA=0
       maskB=0
       maskG=0
       maskR=0
       outputIndex=0
      }
     }]
    convertOutputs [{type=Scalar}, {type=Scalar}]
   }
  }
  MaterialExpressionConvert_7 {
   class=MaterialExpressionConvert
   inputs [
    None {from=MaterialExpressionMultiply_13}
    None {from=MaterialExpressionMultiply_14}
    None {from=MaterialExpressionViewProperty_5, output=Property}
    None {from=""}
   ]
   properties {
    nodeName=Make Vector4
    convertInputs [
     {
      type=Scalar
      expressionInput {
       expression=&self:MaterialExpressionMultiply_13
       inputName=None
       mask=0
       maskA=0
       maskB=0
       maskG=0
       maskR=0
       outputIndex=0
      }
     }
     {
      type=Scalar
      expressionInput {
       expression=&self:MaterialExpressionMultiply_14
       inputName=None
       mask=0
       maskA=0
       maskB=0
       maskG=0
       maskR=0
       outputIndex=0
      }
     }
     {
      type=Scalar
      expressionInput {
       expression=&self:MaterialExpressionViewProperty_5
       inputName=None
       mask=0
       maskA=0
       maskB=0
       maskG=0
       maskR=0
       outputIndex=0
      }
     }
     {
      type=Scalar
      defaultValue=<rgba 1 0 0 1>
      expressionInput {expression=null, inputName=None, mask=0, maskA=0, maskB=0, maskG=0, maskR=0, outputIndex=0}
     }
    ]
    convertOutputs [{type=Vector4, defaultValue=<rgba 0 0 0 1>}]
   }
  }
  MaterialExpressionCustom_6 {
   class=MaterialExpressionCustom
   inputs [In {from=MaterialExpressionConvert_7, output=""}]
   properties {
    code="mul(In,LWCToFloat(ResolvedView.ScreenToWorld))"
    desc="\n*** Original source before expression upgrade ***\nmul(float4(In.xyz,1),View.WorldToClip)"
    outputType=CMOT_Float4
    showCode=true
    inputs [{
      inputName=In
      input {
       expression=&self:MaterialExpressionConvert_7
       inputName=None
       mask=0
       maskA=0
       maskB=0
       maskG=0
       maskR=0
       outputIndex=0
      }
     }]
   }
  }
  MaterialExpressionFunctionOutput_0 {class=MaterialExpressionFunctionOutput, inputs [None {from=MaterialExpressionComponentMask_4, output=""}]}
  MaterialExpressionMultiply_13 {
   class=MaterialExpressionMultiply
   inputs [
    A {from=MaterialExpressionViewProperty_5, output=Property}
    B {from=MaterialExpressionMultiply_6}
   ]
  }
  MaterialExpressionMultiply_14 {
   class=MaterialExpressionMultiply
   inputs [
    A {from=MaterialExpressionViewProperty_5, output=Property}
    B {from=MaterialExpressionSubtract_5}
   ]
  }
  MaterialExpressionMultiply_6 {
   class=MaterialExpressionMultiply
   inputs [
    A {from=MaterialExpressionSubtract_4}
    B {from=""}
   ]
   properties {constB=2}
  }
  MaterialExpressionMultiply_7 {
   class=MaterialExpressionMultiply
   inputs [
    A {from=MaterialExpressionConvert_0}
    B {from=""}
   ]
   properties {constB=2}
  }
  MaterialExpressionScreenPosition_2 {class=MaterialExpressionScreenPosition, inputs [], properties {}}
  MaterialExpressionSubtract_4 {
   class=MaterialExpressionSubtract
   inputs [
    A {from=MaterialExpressionConvert_0}
    B {from=""}
   ]
   properties {constB=0.5}
  }
  MaterialExpressionSubtract_5 {
   class=MaterialExpressionSubtract
   inputs [
    A {from=""}
    B {from=MaterialExpressionMultiply_7}
   ]
   properties {constB=0.5}
  }
  MaterialExpressionViewProperty_5 {class=MaterialExpressionViewProperty, inputs [], properties {property=MEVP_NearPlane}}
 }
 properties {bExposeToLibrary=true}
}
