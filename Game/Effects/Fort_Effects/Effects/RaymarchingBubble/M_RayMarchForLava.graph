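# Material that raymarches a blobby bubble effect, named for lava, from signed distance fields built in Custom HLSL nodes.
# Vector parameters supply five sets of bubbles, capsules and ground data to the SDF, with smoothing and box inputs.
# The march uses pixel-radius thresholds, scene-depth termination, tetrahedral normals and a closest-approach fallback for soft edges.
# Optional ground cylinders fade tails, and scene color is sampled separately per RGB channel.
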Material /Game/Effects/Fort_Effects/Effects/RaymarchingBubble/M_RayMarchForLava
graph {
 expressions {
  MaterialExpressionAbs_1 {class=MaterialExpressionAbs, inputs [None {from=MaterialExpressionDotProduct_1, output=""}], properties {}}
  MaterialExpressionActorPositionWS_0 {class=MaterialExpressionActorPositionWS, inputs [], properties {}}
  MaterialExpressionActorPositionWS_1 {class=MaterialExpressionActorPositionWS, inputs [], properties {}}
  MaterialExpressionActorPositionWS_2 {class=MaterialExpressionActorPositionWS, inputs [], properties {}}
  MaterialExpressionActorPositionWS_3 {class=MaterialExpressionActorPositionWS, inputs [], properties {}}
  MaterialExpressionAdd_0 {
   class=MaterialExpressionAdd
   inputs [
    A {from=MaterialExpressionMultiply_5}
    B {from=MaterialExpressionCameraPositionWS_3}
   ]
  }
  MaterialExpressionAdd_1 {
   class=MaterialExpressionAdd
   inputs [
    A {from=MaterialExpressionVectorParameter_0, output=R}
    B {from=MaterialExpressionVectorParameter_0, output=R}
   ]
  }
  MaterialExpressionAdd_10 {
   class=MaterialExpressionAdd
   inputs [
    A {from=MaterialExpressionVectorParameter_27, output=R}
    B {from=MaterialExpressionVectorParameter_27, output=R}
   ]
  }
  MaterialExpressionAdd_11 {
   class=MaterialExpressionAdd
   inputs [
    A {from=MaterialExpressionVectorParameter_28, output=R}
    B {from=MaterialExpressionVectorParameter_28, output=R}
   ]
  }
  MaterialExpressionAdd_12 {
   class=MaterialExpressionAdd
   inputs [
    A {from=MaterialExpressionMaterialFunctionCall_7, output=Result}
    B {from=MaterialExpressionCustom_5}
   ]
  }
  MaterialExpressionAdd_13 {
   class=MaterialExpressionAdd
   inputs [
    A {from=MaterialExpressionAdd_12}
    B {from=MaterialExpressionMaterialFunctionCall_8, output=Result}
   ]
  }
  MaterialExpressionAdd_14 {
   class=MaterialExpressionAdd
   inputs [
    A {from=MaterialExpressionNamedRerouteUsage_26}
    B {from=MaterialExpressionMaterialFunctionCall_11, output=Result}
   ]
  }
  MaterialExpressionAdd_15 {
   class=MaterialExpressionAdd
   inputs [
    A {from=MaterialExpressionSubtract_6}
    B {from=MaterialExpressionDivide_7}
   ]
  }
  MaterialExpressionAdd_16 {
   class=MaterialExpressionAdd
   inputs [
    A {from=MaterialExpressionSubtract_6}
    B {from=MaterialExpressionDivide_8}
   ]
  }
  MaterialExpressionAdd_17 {
   class=MaterialExpressionAdd
   inputs [
    A {from=MaterialExpressionSubtract_6}
    B {from=MaterialExpressionDivide_9}
   ]
  }
  MaterialExpressionAdd_18 {
   class=MaterialExpressionAdd
   inputs [
    A {from=MaterialExpressionTextureSampleParameter2D_0, output=RGB}
    B {from=MaterialExpressionMultiply_28}
   ]
  }
  MaterialExpressionAdd_19 {
   class=MaterialExpressionAdd
   inputs [
    A {from=MaterialExpressionMultiply_27}
    B {from=MaterialExpressionConstant_7}
   ]
  }
  MaterialExpressionAdd_20 {
   class=MaterialExpressionAdd
   inputs [
    A {from=MaterialExpressionOneMinus_2}
    B {from=MaterialExpressionConstant_22}
   ]
  }
  MaterialExpressionAdd_21 {
   class=MaterialExpressionAdd
   inputs [
    A {from=MaterialExpressionMultiply_25}
    B {from=MaterialExpressionMultiply_26}
   ]
  }
  MaterialExpressionAdd_22 {
   class=MaterialExpressionAdd
   inputs [
    A {from=MaterialExpressionAdd_14}
    B {from=MaterialExpressionAdd_18}
   ]
  }
  MaterialExpressionAdd_23 {
   class=MaterialExpressionAdd
   inputs [
    A {from=MaterialExpressionOneMinus_2}
    B {from=MaterialExpressionConstant_26}
   ]
  }
  MaterialExpressionAdd_24 {
   class=MaterialExpressionAdd
   inputs [
    A {from=MaterialExpressionMultiply_31}
    B {from=MaterialExpressionAdd_22}
   ]
  }
  MaterialExpressionAdd_25 {
   class=MaterialExpressionAdd
   inputs [
    A {from=MaterialExpressionDotProduct_2}
    B {from=MaterialExpressionConstant_31}
   ]
  }
  MaterialExpressionAdd_26 {
   class=MaterialExpressionAdd
   inputs [
    A {from=MaterialExpressionDotProduct_2}
    B {from=MaterialExpressionConstant_34}
   ]
  }
  MaterialExpressionAdd_27 {
   class=MaterialExpressionAdd
   inputs [
    A {from=MaterialExpressionDivide_10}
    B {from=MaterialExpressionConstant_62}
   ]
  }
  MaterialExpressionAdd_3 {
   class=MaterialExpressionAdd
   inputs [
    A {from=MaterialExpressionVectorParameter_25, output=R}
    B {from=MaterialExpressionVectorParameter_25, output=R}
   ]
  }
  MaterialExpressionAdd_4 {
   class=MaterialExpressionAdd
   inputs [
    A {from=MaterialExpressionVectorParameter_0, output=RGB}
    B {from=MaterialExpressionMultiply_15}
   ]
  }
  MaterialExpressionAdd_5 {
   class=MaterialExpressionAdd
   inputs [
    A {from=MaterialExpressionVectorParameter_25, output=RGB}
    B {from=MaterialExpressionMultiply_14}
   ]
  }
  MaterialExpressionAdd_6 {
   class=MaterialExpressionAdd
   inputs [
    A {from=MaterialExpressionVectorParameter_26, output=RGB}
    B {from=MaterialExpressionMultiply_16}
   ]
  }
  MaterialExpressionAdd_7 {
   class=MaterialExpressionAdd
   inputs [
    A {from=MaterialExpressionVectorParameter_27, output=RGB}
    B {from=MaterialExpressionMultiply_17}
   ]
  }
  MaterialExpressionAdd_8 {
   class=MaterialExpressionAdd
   inputs [
    A {from=MaterialExpressionVectorParameter_28, output=RGB}
    B {from=MaterialExpressionMultiply_18}
   ]
  }
  MaterialExpressionAdd_9 {
   class=MaterialExpressionAdd
   inputs [
    A {from=MaterialExpressionVectorParameter_26, output=R}
    B {from=MaterialExpressionVectorParameter_26, output=R}
   ]
  }
  MaterialExpressionCameraPositionWS_1 {class=MaterialExpressionCameraPositionWS, inputs [], properties {}}
  MaterialExpressionCameraPositionWS_12 {class=MaterialExpressionCameraPositionWS, inputs [], properties {}}
  MaterialExpressionCameraPositionWS_13 {class=MaterialExpressionCameraPositionWS, inputs [], properties {}}
  MaterialExpressionCameraPositionWS_14 {class=MaterialExpressionCameraPositionWS, inputs [], properties {}}
  MaterialExpressionCameraPositionWS_15 {class=MaterialExpressionCameraPositionWS, inputs [], properties {}}
  MaterialExpressionCameraPositionWS_16 {class=MaterialExpressionCameraPositionWS, inputs [], properties {}}
  MaterialExpressionCameraPositionWS_2 {class=MaterialExpressionCameraPositionWS, inputs [], properties {}}
  MaterialExpressionCameraPositionWS_3 {class=MaterialExpressionCameraPositionWS, inputs [], properties {}}
  MaterialExpressionCameraPositionWS_4 {class=MaterialExpressionCameraPositionWS, inputs [], properties {}}
  MaterialExpressionCameraPositionWS_7 {class=MaterialExpressionCameraPositionWS, inputs [], properties {}}
  MaterialExpressionCameraPositionWS_9 {class=MaterialExpressionCameraPositionWS, inputs [], properties {}}
  MaterialExpressionCameraVectorWS_0 {class=MaterialExpressionCameraVectorWS, inputs [], properties {}}
  MaterialExpressionCameraVectorWS_1 {class=MaterialExpressionCameraVectorWS, inputs [], properties {}}
  MaterialExpressionCeil_0 {class=MaterialExpressionCeil, inputs [None {from=MaterialExpressionAdd_1, output=""}], properties {}}
  MaterialExpressionCeil_1 {class=MaterialExpressionCeil, inputs [None {from=MaterialExpressionMaterialFunctionCall_23, output=T1}], properties {}}
  MaterialExpressionCeil_2 {class=MaterialExpressionCeil, inputs [None {from=MaterialExpressionMaterialFunctionCall_24, output=T1}], properties {}}
  MaterialExpressionCeil_3 {class=MaterialExpressionCeil, inputs [None {from=MaterialExpressionMaterialFunctionCall_25, output=T1}], properties {}}
  MaterialExpressionCeil_4 {class=MaterialExpressionCeil, inputs [None {from=MaterialExpressionMaterialFunctionCall_26, output=T1}], properties {}}
  MaterialExpressionCeil_5 {class=MaterialExpressionCeil, inputs [None {from=MaterialExpressionMaterialFunctionCall_22, output=T1}], properties {}}
  MaterialExpressionCeil_6 {class=MaterialExpressionCeil, inputs [None {from=MaterialExpressionAdd_3, output=""}], properties {}}
  MaterialExpressionCeil_7 {class=MaterialExpressionCeil, inputs [None {from=MaterialExpressionAdd_9, output=""}], properties {}}
  MaterialExpressionCeil_8 {class=MaterialExpressionCeil, inputs [None {from=MaterialExpressionAdd_10, output=""}], properties {}}
  MaterialExpressionCeil_9 {class=MaterialExpressionCeil, inputs [None {from=MaterialExpressionAdd_11, output=""}], properties {}}
  MaterialExpressionComponentMask_0 {class=MaterialExpressionComponentMask, inputs [None {from=MaterialExpressionCustom_0, output=return}], properties {b=true, g=true, r=true}}
  MaterialExpressionComponentMask_1 {class=MaterialExpressionComponentMask, inputs [None {from=MaterialExpressionCustom_0, output=return}], properties {a=true}}
  MaterialExpressionComponentMask_2 {class=MaterialExpressionComponentMask, inputs [None {from=MaterialExpressionViewSize_1}], properties {g=true}}
  MaterialExpressionComponentMask_3 {class=MaterialExpressionComponentMask, inputs [None {from=MaterialExpressionMultiply_0}], properties {r=true}}
  MaterialExpressionComponentMask_7 {class=MaterialExpressionComponentMask, inputs [None {from=MaterialExpressionSceneColor_5, output=RGB}], properties {r=true}}
  MaterialExpressionComponentMask_8 {class=MaterialExpressionComponentMask, inputs [None {from=MaterialExpressionSceneColor_4, output=RGB}], properties {g=true}}
  MaterialExpressionComponentMask_9 {class=MaterialExpressionComponentMask, inputs [None {from=MaterialExpressionSceneColor_3, output=RGB}], properties {b=true}}
  MaterialExpressionConstant3Vector_2 {class=MaterialExpressionConstant3Vector, properties {constant=<rgba 0 0 1 0>}}
  MaterialExpressionConstant3Vector_3 {class=MaterialExpressionConstant3Vector, properties {constant=<rgba 0 0 10000 0>}}
  MaterialExpressionConstant3Vector_4 {class=MaterialExpressionConstant3Vector, properties {constant=<rgba 256 256 256 0>}}
  MaterialExpressionConstant3Vector_5 {class=MaterialExpressionConstant3Vector, properties {constant=<rgba 0 0 1 0>}}
  MaterialExpressionConstant4Vector_2 {class=MaterialExpressionConstant4Vector, inputs [], properties {constant=<rgba 0.2 0.2 0.2 1>}}
  MaterialExpressionConstant4Vector_3 {class=MaterialExpressionConstant4Vector, inputs [], properties {constant=<rgba 0.0105 0.01 0.0095 1>}}
  MaterialExpressionConstant4Vector_4 {class=MaterialExpressionConstant4Vector, inputs [], properties {constant=<rgba 0 0 0 1>}}
  MaterialExpressionConstant4Vector_5 {class=MaterialExpressionConstant4Vector, inputs [], properties {constant=<rgba 5 0.01 0 1>}}
  MaterialExpressionConstant4Vector_6 {class=MaterialExpressionConstant4Vector, inputs [], properties {constant=<rgba 2 0 0 1>}}
  MaterialExpressionConstant_0 {class=MaterialExpressionConstant}
  MaterialExpressionConstant_1 {class=MaterialExpressionConstant}
  MaterialExpressionConstant_10 {class=MaterialExpressionConstant}
  MaterialExpressionConstant_11 {class=MaterialExpressionConstant, properties {r=1}}
  MaterialExpressionConstant_12 {class=MaterialExpressionConstant}
  MaterialExpressionConstant_13 {class=MaterialExpressionConstant, properties {r=1}}
  MaterialExpressionConstant_14 {class=MaterialExpressionConstant}
  MaterialExpressionConstant_15 {class=MaterialExpressionConstant, properties {r=1}}
  MaterialExpressionConstant_16 {class=MaterialExpressionConstant}
  MaterialExpressionConstant_17 {class=MaterialExpressionConstant, properties {r=1}}
  MaterialExpressionConstant_18 {class=MaterialExpressionConstant}
  MaterialExpressionConstant_19 {class=MaterialExpressionConstant, properties {r=1}}
  MaterialExpressionConstant_2 {class=MaterialExpressionConstant, properties {r=1}}
  MaterialExpressionConstant_20 {class=MaterialExpressionConstant, properties {r=1.5}}
  MaterialExpressionConstant_21 {class=MaterialExpressionConstant, properties {r=0.5}}
  MaterialExpressionConstant_22 {class=MaterialExpressionConstant, properties {r=0.2}}
  MaterialExpressionConstant_23 {class=MaterialExpressionConstant, properties {r=1.5}}
  MaterialExpressionConstant_24 {class=MaterialExpressionConstant, properties {r=1.5}}
  MaterialExpressionConstant_25 {class=MaterialExpressionConstant, properties {r=0.4}}
  MaterialExpressionConstant_26 {class=MaterialExpressionConstant}
  MaterialExpressionConstant_27 {class=MaterialExpressionConstant, properties {r=1}}
  MaterialExpressionConstant_28 {class=MaterialExpressionConstant, properties {r=2.5}}
  MaterialExpressionConstant_3 {class=MaterialExpressionConstant, properties {r=4}}
  MaterialExpressionConstant_30 {class=MaterialExpressionConstant}
  MaterialExpressionConstant_31 {class=MaterialExpressionConstant, properties {r=-1.0301}}
  MaterialExpressionConstant_32 {class=MaterialExpressionConstant, properties {r=2}}
  MaterialExpressionConstant_33 {class=MaterialExpressionConstant, properties {r=1}}
  MaterialExpressionConstant_34 {class=MaterialExpressionConstant, properties {r=1.4}}
  MaterialExpressionConstant_35 {class=MaterialExpressionConstant, properties {r=1}}
  MaterialExpressionConstant_36 {class=MaterialExpressionConstant, properties {r=1}}
  MaterialExpressionConstant_37 {class=MaterialExpressionConstant, properties {r=0.6}}
  MaterialExpressionConstant_38 {class=MaterialExpressionConstant, properties {r=0.4}}
  MaterialExpressionConstant_40 {class=MaterialExpressionConstant, properties {r=1}}
  MaterialExpressionConstant_41 {class=MaterialExpressionConstant, properties {r=1.05}}
  MaterialExpressionConstant_42 {class=MaterialExpressionConstant}
  MaterialExpressionConstant_43 {class=MaterialExpressionConstant, properties {r=1}}
  MaterialExpressionConstant_44 {class=MaterialExpressionConstant, properties {r=2.25}}
  MaterialExpressionConstant_45 {class=MaterialExpressionConstant, properties {r=1.8}}
  MaterialExpressionConstant_46 {class=MaterialExpressionConstant, properties {r=1.5}}
  MaterialExpressionConstant_47 {class=MaterialExpressionConstant, properties {r=7}}
  MaterialExpressionConstant_48 {class=MaterialExpressionConstant, properties {r=6}}
  MaterialExpressionConstant_49 {class=MaterialExpressionConstant, properties {r=5}}
  MaterialExpressionConstant_50 {
   class=MaterialExpressionConstant
   properties {desc="if the marched point is below this distance from the df it is counted as being within the df. ", r=0.5}
  }
  MaterialExpressionConstant_51 {class=MaterialExpressionConstant, properties {r=18}}
  MaterialExpressionConstant_52 {class=MaterialExpressionConstant, properties {r=10}}
  MaterialExpressionConstant_53 {class=MaterialExpressionConstant, properties {r=10}}
  MaterialExpressionConstant_54 {class=MaterialExpressionConstant, properties {r=1.3}}
  MaterialExpressionConstant_55 {class=MaterialExpressionConstant, properties {r=3}}
  MaterialExpressionConstant_56 {class=MaterialExpressionConstant, properties {r=2}}
  MaterialExpressionConstant_57 {class=MaterialExpressionConstant, properties {r=1}}
  MaterialExpressionConstant_58 {class=MaterialExpressionConstant, properties {r=3}}
  MaterialExpressionConstant_59 {class=MaterialExpressionConstant, properties {r=3}}
  MaterialExpressionConstant_6 {class=MaterialExpressionConstant}
  MaterialExpressionConstant_60 {class=MaterialExpressionConstant, properties {r=3}}
  MaterialExpressionConstant_61 {class=MaterialExpressionConstant, properties {r=3.5}}
  MaterialExpressionConstant_62 {class=MaterialExpressionConstant, properties {r=0.3}}
  MaterialExpressionConstant_63 {class=MaterialExpressionConstant, properties {r=1.5}}
  MaterialExpressionConstant_64 {class=MaterialExpressionConstant, properties {r=1.3}}
  MaterialExpressionConstant_65 {class=MaterialExpressionConstant, properties {r=128}}
  MaterialExpressionConstant_66 {class=MaterialExpressionConstant}
  MaterialExpressionConstant_67 {class=MaterialExpressionConstant, properties {r=300}}
  MaterialExpressionConstant_68 {class=MaterialExpressionConstant}
  MaterialExpressionConstant_69 {class=MaterialExpressionConstant}
  MaterialExpressionConstant_7 {class=MaterialExpressionConstant, properties {r=0.25}}
  MaterialExpressionConstant_8 {class=MaterialExpressionConstant, properties {r=0.001}}
  MaterialExpressionConstant_9 {class=MaterialExpressionConstant, properties {r=1}}
  MaterialExpressionConvert_0 {
   class=MaterialExpressionConvert
   inputs [None {from=MaterialExpressionSubtract_3}]
   properties {
    nodeName=Break Float3
    convertInputs [{
      type=Vector3
      expressionInput {
       expression=&self:MaterialExpressionSubtract_3
       inputName=None
       mask=0
       maskA=0
       maskB=0
       maskG=0
       maskR=0
       outputIndex=0
      }
     }]
    convertOutputs [{type=Scalar}, {type=Scalar}, {type=Scalar}]
   }
  }
  MaterialExpressionConvert_1 {
   class=MaterialExpressionConvert
   inputs [
    None {from=MaterialExpressionMax_5}
    None {from=MaterialExpressionMax_5}
    None {from=MaterialExpressionMax_5}
   ]
   properties {
    nodeName=Make Vector3
    convertInputs [
     {
      type=Scalar
      expressionInput {
       expression=&self:MaterialExpressionMax_5
       inputName=None
       mask=0
       maskA=0
       maskB=0
       maskG=0
       maskR=0
       outputIndex=0
      }
     }
     {
      type=Scalar
      expressionInput {
       expression=&self:MaterialExpressionMax_5
       inputName=None
       mask=0
       maskA=0
       maskB=0
       maskG=0
       maskR=0
       outputIndex=0
      }
     }
     {
      type=Scalar
      expressionInput {
       expression=&self:MaterialExpressionMax_5
       inputName=None
       mask=0
       maskA=0
       maskB=0
       maskG=0
       maskR=0
       outputIndex=0
      }
     }
    ]
    convertOutputs [{type=Vector3}]
   }
  }
  MaterialExpressionConvert_3 {
   class=MaterialExpressionConvert
   inputs [
    None {from=MaterialExpressionMax_6}
    None {from=MaterialExpressionMax_6}
    None {from=MaterialExpressionMax_6}
   ]
   properties {
    nodeName=Make Vector3
    convertInputs [
     {
      type=Scalar
      expressionInput {
       expression=&self:MaterialExpressionMax_6
       inputName=None
       mask=0
       maskA=0
       maskB=0
       maskG=0
       maskR=0
       outputIndex=0
      }
     }
     {
      type=Scalar
      expressionInput {
       expression=&self:MaterialExpressionMax_6
       inputName=None
       mask=0
       maskA=0
       maskB=0
       maskG=0
       maskR=0
       outputIndex=0
      }
     }
     {
      type=Scalar
      expressionInput {
       expression=&self:MaterialExpressionMax_6
       inputName=None
       mask=0
       maskA=0
       maskB=0
       maskG=0
       maskR=0
       outputIndex=0
      }
     }
    ]
    convertOutputs [{type=Vector3}]
   }
  }
  MaterialExpressionConvert_4 {
   class=MaterialExpressionConvert
   inputs [
    None {from=MaterialExpressionMax_7}
    None {from=MaterialExpressionMax_7}
    None {from=MaterialExpressionMax_7}
   ]
   properties {
    nodeName=Make Vector3
    convertInputs [
     {
      type=Scalar
      expressionInput {
       expression=&self:MaterialExpressionMax_7
       inputName=None
       mask=0
       maskA=0
       maskB=0
       maskG=0
       maskR=0
       outputIndex=0
      }
     }
     {
      type=Scalar
      expressionInput {
       expression=&self:MaterialExpressionMax_7
       inputName=None
       mask=0
       maskA=0
       maskB=0
       maskG=0
       maskR=0
       outputIndex=0
      }
     }
     {
      type=Scalar
      expressionInput {
       expression=&self:MaterialExpressionMax_7
       inputName=None
       mask=0
       maskA=0
       maskB=0
       maskG=0
       maskR=0
       outputIndex=0
      }
     }
    ]
    convertOutputs [{type=Vector3}]
   }
  }
  MaterialExpressionConvert_5 {
   class=MaterialExpressionConvert
   inputs [
    None {from=MaterialExpressionMax_8}
    None {from=MaterialExpressionMax_8}
    None {from=MaterialExpressionMax_8}
   ]
   properties {
    nodeName=Make Vector3
    convertInputs [
     {
      type=Scalar
      expressionInput {
       expression=&self:MaterialExpressionMax_8
       inputName=None
       mask=0
       maskA=0
       maskB=0
       maskG=0
       maskR=0
       outputIndex=0
      }
     }
     {
      type=Scalar
      expressionInput {
       expression=&self:MaterialExpressionMax_8
       inputName=None
       mask=0
       maskA=0
       maskB=0
       maskG=0
       maskR=0
       outputIndex=0
      }
     }
     {
      type=Scalar
      expressionInput {
       expression=&self:MaterialExpressionMax_8
       inputName=None
       mask=0
       maskA=0
       maskB=0
       maskG=0
       maskR=0
       outputIndex=0
      }
     }
    ]
    convertOutputs [{type=Vector3}]
   }
  }
  MaterialExpressionConvert_6 {
   class=MaterialExpressionConvert
   inputs [
    None {from=MaterialExpressionMax_9}
    None {from=MaterialExpressionMax_9}
    None {from=MaterialExpressionMax_9}
   ]
   properties {
    nodeName=Make Vector3
    convertInputs [
     {
      type=Scalar
      expressionInput {
       expression=&self:MaterialExpressionMax_9
       inputName=None
       mask=0
       maskA=0
       maskB=0
       maskG=0
       maskR=0
       outputIndex=0
      }
     }
     {
      type=Scalar
      expressionInput {
       expression=&self:MaterialExpressionMax_9
       inputName=None
       mask=0
       maskA=0
       maskB=0
       maskG=0
       maskR=0
       outputIndex=0
      }
     }
     {
      type=Scalar
      expressionInput {
       expression=&self:MaterialExpressionMax_9
       inputName=None
       mask=0
       maskA=0
       maskB=0
       maskG=0
       maskR=0
       outputIndex=0
      }
     }
    ]
    convertOutputs [{type=Vector3}]
   }
  }
  MaterialExpressionCustom_0 {
   class=MaterialExpressionCustom
   inputs [
    Custom0 {from=MaterialExpressionCustom_2, output=""}
    StartPos {from=MaterialExpressionConstant_6, output=""}
    WorldPos {from=MaterialExpressionCameraPositionWS_7, output=""}
    CameraPos {from=MaterialExpressionCameraPositionWS_7, output=""}
    CameraVector {from=MaterialExpressionNamedRerouteUsage_3, output=""}
    thresh {from=MaterialExpressionNamedRerouteUsage_12, output=""}
    epsilon {from=MaterialExpressionNamedRerouteUsage_10, output=""}
    PositionAndSize0 {from=MaterialExpressionVectorParameter_4, output=RGBA}
    PositionAndSize1 {from=MaterialExpressionVectorParameter_5, output=RGBA}
    PositionAndSize2 {from=MaterialExpressionVectorParameter_6, output=RGBA}
    PositionAndSize3 {from=MaterialExpressionVectorParameter_7, output=RGBA}
    PositionAndSize4 {from=MaterialExpressionVectorParameter_8, output=RGBA}
    Capsule0 {from=MaterialExpressionVectorParameter_9, output=RGBA}
    Capsule1 {from=MaterialExpressionVectorParameter_10, output=RGBA}
    Capsule2 {from=MaterialExpressionVectorParameter_11, output=RGBA}
    Capsule3 {from=MaterialExpressionVectorParameter_12, output=RGBA}
    Capsule4 {from=MaterialExpressionVectorParameter_13, output=RGBA}
    SecondBubble0 {from=MaterialExpressionVectorParameter_14, output=RGBA}
    SecondBubble1 {from=MaterialExpressionVectorParameter_15, output=RGBA}
    SecondBubble2 {from=MaterialExpressionVectorParameter_16, output=RGBA}
    SecondBubble3 {from=MaterialExpressionVectorParameter_17, output=RGBA}
    SecondBubble4 {from=MaterialExpressionVectorParameter_18, output=RGBA}
    Ground0 {from=MaterialExpressionVectorParameter_29, output=RGBA}
    Ground1 {from=MaterialExpressionVectorParameter_30, output=RGBA}
    Ground2 {from=MaterialExpressionVectorParameter_31, output=RGBA}
    Ground3 {from=MaterialExpressionVectorParameter_32, output=RGBA}
    Ground4 {from=MaterialExpressionVectorParameter_33, output=RGBA}
    maxSteps {from=MaterialExpressionNamedRerouteUsage_13, output=""}
    maxDist {from=MaterialExpressionConstant_67, output=""}
    smoothK {from=MaterialExpressionVectorParameter_21, output=RGBA}
    BoxCenter {from=MaterialExpressionSubtract_1, output=""}
    BoxSize {from=MaterialExpressionVectorParameter_20, output=RGB}
    distToObject {from=MaterialExpressionNamedRerouteUsage_1, output=""}
    pixelScale {from=MaterialExpressionNamedRerouteUsage_0, output=""}
    SceneDepth {from=MaterialExpressionNamedRerouteUsage_2, output=""}
    EdgeWidth {from=MaterialExpressionNamedRerouteUsage_11, output=""}
   ]
   properties {
    code="\r\n\r\nfloat3 curpos = WorldPos;\r\nfloat curdist = StartPos;\r\nfloat3 nn = float3(0,0,0);\r\nfloat totalDist = 0;\r\nfloat minCandidateDist = 1000000.0;\r\nfloat3 minCandidatePos = curpos;\r\nfloat4 result = float4(0,0,0,0);\r\nbool bHit = false;\r\nfloat4 PositionAndSize[5];\r\nfloat4 Capsule[5];\r\nfloat4 SecondBubble[5];\r\nfloat4 Grounds[5];\r\nPositionAndSize[0] = PositionAndSize0;\r\nPositionAndSize[1] = PositionAndSize1;\r\nPositionAndSize[2] = PositionAndSize2;\r\nPositionAndSize[3] = PositionAndSize3;\r\nPositionAndSize[4] = PositionAndSize4;\r\nCapsule[0] = Capsule0;\r\nCapsule[1] = Capsule1;\r\nCapsule[2] = Capsule2;\r\nCapsule[3] = Capsule3;\r\nCapsule[4] = Capsule4;\r\nSecondBubble[0] = SecondBubble0;\r\nSecondBubble[1] = SecondBubble1;\r\nSecondBubble[2] = SecondBubble2;\r\nSecondBubble[3] = SecondBubble3;\r\nSecondBubble[4] = SecondBubble4;\r\nGrounds[0] = Ground0;\r\nGrounds[1] = Ground1;\r\nGrounds[2] = Ground2;\r\nGrounds[3] = Ground3;\r\nGrounds[4] = Ground4;\r\n\r\nfor (int i = 0; i < maxSteps; ++i)\r\n{\r\n    numSteps = i + 1;\r\n\r\n    curdist = SDFFunction(curpos, PositionAndSize, Capsule, SecondBubble,Grounds, smoothK, BoxCenter, BoxSize);\r\n\r\n    if (curdist < minCandidateDist)\r\n    {\r\n        minCandidateDist = curdist;\r\n        minCandidatePos = curpos;\r\n    }\r\n\r\n    float currentPixelRadius = (totalDist - distToObject) * pixelScale;\r\n\r\n    if (curdist < thresh || curdist < currentPixelRadius)\r\n    {\r\n        Dist = totalDist;\r\n\r\n        // Tetrahedron normal\r\n        const float2 k = float2(1.0, -1.0);\r\n        nn = normalize(\r\n            k.xyy * SDFFunction(curpos + k.xyy * epsilon, PositionAndSize, Capsule, SecondBubble, Grounds, smoothK, BoxCenter, BoxSize) +\r\n            k.yyx * SDFFunction(curpos + k.yyx * epsilon, PositionAndSize, Capsule, SecondBubble, Grounds, smoothK, BoxCenter, BoxSize) +\r\n            k.yxy * SDFFunction(curpos + k.yxy * epsilon, PositionAndSize, Capsule, SecondBubble, Grounds, smoothK, BoxCenter, BoxSize) +\r\n            k.xxx * SDFFunction(curpos + k.xxx * epsilon, PositionAndSize, Capsule, SecondBubble, Grounds, smoothK, BoxCenter, BoxSize)\r\n        );\r\n\r\n        distToGround = 1e20;\r\n// If hit, check the distance to the ground SDF for the fade-out effect.\r\n        ##if FEATURE_GROUNDFADE\r\n        distToGround = GroundSDF(curpos, PositionAndSize,Capsule, Grounds);\r\n        ##endif\r\n        result = float4(nn.x, nn.y, nn.z, 1);\r\n        bHit = true;\r\n        break;\r\n    }\r\n// StepSize should be > PixelRadius. If StepSize < PixelRadius, the step is ineffective.\r\n    float stepSize = max(curdist, currentPixelRadius);\r\n\r\n    curpos += CameraVector * stepSize;\r\n    totalDist += stepSize;\r\n\r\n    if (totalDist - distToObject > maxDist || SceneDepth < totalDist)\r\n    {\r\n        break;\r\n    }\r\n}\r\n\r\n\r\nif (!bHit &&&& SceneDepth >= totalDist)\r\n{\r\n// If nothing is hit, but the distance is within range, render it.\r\n// This resolves raymarching occlusion issues on shape edges (especially when the loop count is low).\r\n    if (minCandidateDist < EdgeWidth &&&& totalDist - distToObject < maxDist)\r\n    {\r\n        float alpha = 1.0 - smoothstep(0.0, EdgeWidth, minCandidateDist);\r\n\r\n        float d_here = minCandidateDist;\r\n        //nn.x = SDFFunction(minCandidatePos + float3(epsilon, 0, 0), PositionAndSize, Capsule, SecondBubble,Grounds, smoothK, BoxCenter, BoxSize) - d_here;\r\n        //nn.y = SDFFunction(minCandidatePos + float3(0, epsilon, 0), PositionAndSize, Capsule, SecondBubble,Grounds, smoothK, BoxCenter, BoxSize) - d_here;\r\n        //nn.z = SDFFunction(minCandidatePos + float3(0, 0, epsilon), PositionAndSize, Capsule, SecondBubble,Grounds, smoothK, BoxCenter, BoxSize) - d_here;\r\n        nn.x = 0;\r\n        nn.y = 0;\r\n        nn.z = 1;\r\n        //nn = normalize(nn);\r\n        Dist = totalDist;\r\n        curpos = minCandidatePos;\r\n        distToGround = 1e20;\r\n        ##if FEATURE_GROUNDFADE\r\n        //distToGround = GroundSDF(curpos, PositionAndSize,Capsule, Grounds);\r\n        ##endif\r\n        result = float4(nn.x, nn.y, nn.z, alpha);\r\n    }\r\n}\r\n\r\nreturn result;"
    description=RayMarch
    outputType=CMOT_Float4
    showCode=true
    additionalOutputs [{outputName=numSteps, outputType=CMOT_Float1}, {outputName=Dist, outputType=CMOT_Float1}, {outputName=distToGround, outputType=CMOT_Float1}]
    inputs [
     {
      inputName=Custom0
      input {
       expression=&self:MaterialExpressionCustom_2
       inputName=None
       mask=0
       maskA=0
       maskB=0
       maskG=0
       maskR=0
       outputIndex=0
      }
     }
     {
      inputName=StartPos
      input {
       expression=&self:MaterialExpressionConstant_6
       inputName=None
       mask=0
       maskA=0
       maskB=0
       maskG=0
       maskR=0
       outputIndex=0
      }
     }
     {
      inputName=WorldPos
      input {
       expression=&self:MaterialExpressionCameraPositionWS_7
       inputName=None
       mask=0
       maskA=0
       maskB=0
       maskG=0
       maskR=0
       outputIndex=0
      }
     }
     {
      inputName=CameraPos
      input {
       expression=&self:MaterialExpressionCameraPositionWS_7
       inputName=None
       mask=0
       maskA=0
       maskB=0
       maskG=0
       maskR=0
       outputIndex=0
      }
     }
     {
      inputName=CameraVector
      input {
       expression=&self:MaterialExpressionNamedRerouteUsage_3
       inputName=None
       mask=0
       maskA=0
       maskB=0
       maskG=0
       maskR=0
       outputIndex=0
      }
     }
     {
      inputName=thresh
      input {
       expression=&self:MaterialExpressionNamedRerouteUsage_12
       inputName=None
       mask=0
       maskA=0
       maskB=0
       maskG=0
       maskR=0
       outputIndex=0
      }
     }
     {
      inputName=epsilon
      input {
       expression=&self:MaterialExpressionNamedRerouteUsage_10
       inputName=None
       mask=0
       maskA=0
       maskB=0
       maskG=0
       maskR=0
       outputIndex=0
      }
     }
     {
      inputName=PositionAndSize0
      input {
       expression=&self:MaterialExpressionVectorParameter_4
       inputName=None
       mask=1
       maskA=1
       maskB=1
       maskG=1
       maskR=1
       outputIndex=5
      }
     }
     {
      inputName=PositionAndSize1
      input {
       expression=&self:MaterialExpressionVectorParameter_5
       inputName=None
       mask=1
       maskA=1
       maskB=1
       maskG=1
       maskR=1
       outputIndex=5
      }
     }
     {
      inputName=PositionAndSize2
      input {
       expression=&self:MaterialExpressionVectorParameter_6
       inputName=None
       mask=1
       maskA=1
       maskB=1
       maskG=1
       maskR=1
       outputIndex=5
      }
     }
     {
      inputName=PositionAndSize3
      input {
       expression=&self:MaterialExpressionVectorParameter_7
       inputName=None
       mask=1
       maskA=1
       maskB=1
       maskG=1
       maskR=1
       outputIndex=5
      }
     }
     {
      inputName=PositionAndSize4
      input {
       expression=&self:MaterialExpressionVectorParameter_8
       inputName=None
       mask=1
       maskA=1
       maskB=1
       maskG=1
       maskR=1
       outputIndex=5
      }
     }
     {
      inputName=Capsule0
      input {
       expression=&self:MaterialExpressionVectorParameter_9
       inputName=None
       mask=1
       maskA=1
       maskB=1
       maskG=1
       maskR=1
       outputIndex=5
      }
     }
     {
      inputName=Capsule1
      input {
       expression=&self:MaterialExpressionVectorParameter_10
       inputName=None
       mask=1
       maskA=1
       maskB=1
       maskG=1
       maskR=1
       outputIndex=5
      }
     }
     {
      inputName=Capsule2
      input {
       expression=&self:MaterialExpressionVectorParameter_11
       inputName=None
       mask=1
       maskA=1
       maskB=1
       maskG=1
       maskR=1
       outputIndex=5
      }
     }
     {
      inputName=Capsule3
      input {
       expression=&self:MaterialExpressionVectorParameter_12
       inputName=None
       mask=1
       maskA=1
       maskB=1
       maskG=1
       maskR=1
       outputIndex=5
      }
     }
     {
      inputName=Capsule4
      input {
       expression=&self:MaterialExpressionVectorParameter_13
       inputName=None
       mask=1
       maskA=1
       maskB=1
       maskG=1
       maskR=1
       outputIndex=5
      }
     }
     {
      inputName=SecondBubble0
      input {
       expression=&self:MaterialExpressionVectorParameter_14
       inputName=None
       mask=1
       maskA=1
       maskB=1
       maskG=1
       maskR=1
       outputIndex=5
      }
     }
     {
      inputName=SecondBubble1
      input {
       expression=&self:MaterialExpressionVectorParameter_15
       inputName=None
       mask=1
       maskA=1
       maskB=1
       maskG=1
       maskR=1
       outputIndex=5
      }
     }
     {
      inputName=SecondBubble2
      input {
       expression=&self:MaterialExpressionVectorParameter_16
       inputName=None
       mask=1
       maskA=1
       maskB=1
       maskG=1
       maskR=1
       outputIndex=5
      }
     }
     {
      inputName=SecondBubble3
      input {
       expression=&self:MaterialExpressionVectorParameter_17
       inputName=None
       mask=1
       maskA=1
       maskB=1
       maskG=1
       maskR=1
       outputIndex=5
      }
     }
     {
      inputName=SecondBubble4
      input {
       expression=&self:MaterialExpressionVectorParameter_18
       inputName=None
       mask=1
       maskA=1
       maskB=1
       maskG=1
       maskR=1
       outputIndex=5
      }
     }
     {
      inputName=Ground0
      input {
       expression=&self:MaterialExpressionVectorParameter_29
       inputName=None
       mask=1
       maskA=1
       maskB=1
       maskG=1
       maskR=1
       outputIndex=5
      }
     }
     {
      inputName=Ground1
      input {
       expression=&self:MaterialExpressionVectorParameter_30
       inputName=None
       mask=1
       maskA=1
       maskB=1
       maskG=1
       maskR=1
       outputIndex=5
      }
     }
     {
      inputName=Ground2
      input {
       expression=&self:MaterialExpressionVectorParameter_31
       inputName=None
       mask=1
       maskA=1
       maskB=1
       maskG=1
       maskR=1
       outputIndex=5
      }
     }
     {
      inputName=Ground3
      input {
       expression=&self:MaterialExpressionVectorParameter_32
       inputName=None
       mask=1
       maskA=1
       maskB=1
       maskG=1
       maskR=1
       outputIndex=5
      }
     }
     {
      inputName=Ground4
      input {
       expression=&self:MaterialExpressionVectorParameter_33
       inputName=None
       mask=1
       maskA=1
       maskB=1
       maskG=1
       maskR=1
       outputIndex=5
      }
     }
     {
      inputName=maxSteps
      input {
       expression=&self:MaterialExpressionNamedRerouteUsage_13
       inputName=None
       mask=0
       maskA=0
       maskB=0
       maskG=0
       maskR=0
       outputIndex=0
      }
     }
     {
      inputName=maxDist
      input {
       expression=&self:MaterialExpressionConstant_67
       inputName=None
       mask=0
       maskA=0
       maskB=0
       maskG=0
       maskR=0
       outputIndex=0
      }
     }
     {
      inputName=smoothK
      input {
       expression=&self:MaterialExpressionVectorParameter_21
       inputName=None
       mask=1
       maskA=1
       maskB=1
       maskG=1
       maskR=1
       outputIndex=5
      }
     }
     {
      inputName=BoxCenter
      input {
       expression=&self:MaterialExpressionSubtract_1
       inputName=None
       mask=0
       maskA=0
       maskB=0
       maskG=0
       maskR=0
       outputIndex=0
      }
     }
     {
      inputName=BoxSize
      input {
       expression=&self:MaterialExpressionVectorParameter_20
       inputName=None
       mask=1
       maskA=0
       maskB=1
       maskG=1
       maskR=1
       outputIndex=0
      }
     }
     {
      inputName=distToObject
      input {
       expression=&self:MaterialExpressionNamedRerouteUsage_1
       inputName=None
       mask=0
       maskA=0
       maskB=0
       maskG=0
       maskR=0
       outputIndex=0
      }
     }
     {
      inputName=pixelScale
      input {
       expression=&self:MaterialExpressionNamedRerouteUsage_0
       inputName=None
       mask=0
       maskA=0
       maskB=0
       maskG=0
       maskR=0
       outputIndex=0
      }
     }
     {
      inputName=SceneDepth
      input {
       expression=&self:MaterialExpressionNamedRerouteUsage_2
       inputName=None
       mask=0
       maskA=0
       maskB=0
       maskG=0
       maskR=0
       outputIndex=0
      }
     }
     {
      inputName=EdgeWidth
      input {
       expression=&self:MaterialExpressionNamedRerouteUsage_11
       inputName=None
       mask=0
       maskA=0
       maskB=0
       maskG=0
       maskR=0
       outputIndex=0
      }
     }
    ]
   }
  }
  MaterialExpressionCustom_1 {
   class=MaterialExpressionCustom
   inputs [A {from=MaterialExpressionConstant_0, output=""}]
   properties {
    code="return A;}\r\n##define FEATURE_TAIL 0\r\nfloat VoidFunc(float a){\r\nreturn a+1;"
    showCode=true
    inputs [{
      inputName=A
      input {
       expression=&self:MaterialExpressionConstant_0
       inputName=None
       mask=0
       maskA=0
       maskB=0
       maskG=0
       maskR=0
       outputIndex=0
      }
     }]
   }
  }
  MaterialExpressionCustom_2 {
   class=MaterialExpressionCustom
   inputs [UnionFunc {from=MaterialExpressionCustom_3, output=""}]
   properties {
    code="return 0;}\r\n\r\nfloat GroundSDF(float3 pos, float4 PositionAndSize[5], float4 Capsule[5], float4 Grounds[5])\r\n{\r\n    //---------- Ground SDF ----------\r\n    // This part generates 5 cylinders behind the particle spawn position to create the tail fade-out effect.\r\n    const float BIG = 1e20;\r\n    float dTotalGrounds = BIG;     \r\n    [unroll]\r\n    for (int i = 0; i < 5; ++i)\r\n    {\r\n        float4 groundInfo = Grounds[i];\r\n        float3 normal = groundInfo.xyz; \r\n        float4 capsuleInfo = Capsule[i];\r\n        float  capsuleHeight    = capsuleInfo.x;\r\n        float  capsuleRadius    = capsuleInfo.y;\r\n        float cylinderHalfHeight = 100.0; \r\n        float cylinderRadius = capsuleRadius*3.0;      \r\n        bool hasCapsule = (capsuleRadius > 1e-4) &&&& (capsuleHeight > 1e-4);\r\n\r\n        if (dot(normal, normal) > 0.001 &&&& hasCapsule) \r\n        {\r\n            float3 groundCenter = float3(PositionAndSize[i].x, PositionAndSize[i].y, groundInfo.w - cylinderHalfHeight);\r\n            float3 p = pos - groundCenter;\r\n            float2 d = float2(length(p.xy) - cylinderRadius, abs(p.z) - cylinderHalfHeight);\r\n            float dThisCylinder = min(max(d.x, d.y), 0.0) + length(max(d, 0.0));\r\n            if (dTotalGrounds == BIG)\r\n            {\r\n                dTotalGrounds = dThisCylinder;\r\n            }\r\n            else\r\n            {\r\n                dTotalGrounds = min(dTotalGrounds, dThisCylinder);\r\n            }\r\n        }\r\n    }\r\n    return dTotalGrounds;\r\n}\r\nfloat SDFFunction(float3 pos,float4 PositionAndSize[5],float4 Capsule[5], float4 SecondBubble[5],float4 Grounds[5],float4 smoothK,float3 BoxCenter,float3 BoxSize)\r\n{\r\n    \r\n    const float BIG = 1e20;\r\n    float dTotalBlob = BIG;\r\n    float dTotalCapsule = BIG;\r\n    const float GFDistance = 40;\r\n    const float GFScale = 3;\r\n    [unroll]\r\n    for (int i = 0; i < 5; ++i)\r\n    {\r\n        //----------Prepare Data----------\r\n        float4 blobPosAndSize = PositionAndSize[i]; \r\n        \r\n        if (abs(blobPosAndSize.x) < 1e-4)\r\n            continue; //No break for GPU optimization\r\n\r\n        float4 secondBlobPosAndSize = SecondBubble[i];\r\n\r\n        float3 blobCenter = blobPosAndSize.xyz;\r\n        float  blobSize = blobPosAndSize.w;\r\n\r\n        float3 secondBlobCenter = secondBlobPosAndSize.xyz;\r\n        float secondBlobRadius = secondBlobPosAndSize.w;\r\n\r\n        float3 toCenter = pos - blobCenter;\r\n        float3 toSecCenter = pos - secondBlobCenter;\r\n        float dSphere = length(toCenter) - blobSize;\r\n        //---------- Blob SDF ----------\r\n        // Use 2 bubbles to create dynamics in Niagara, but this costs more as SmoothUnion is expensive.\r\n        ##if FEATURE_TWOSPHEREBUBBLE\r\n        dSphere = SmoothUnion(dSphere, length(toSecCenter) - secondBlobRadius, smoothK.y);\r\n        ##endif\r\n        //---------- Blobs smooth union ----------\r\n        if (dTotalBlob == BIG)\r\n        {\r\n            dTotalBlob = dSphere;\r\n        }\r\n        else\r\n        {\r\n            dTotalBlob = SmoothUnion(dTotalBlob, dSphere, smoothK.x);\r\n        }\r\n    }\r\n    float dFinal = dTotalBlob;\r\n    ##if FEATURE_TAIL\r\n    [unroll]\r\n    for (int i = 0; i < 5; ++i)\r\n    {\r\n        float4 capsuleInfo = Capsule[i];\r\n        float  capsuleHeight    = capsuleInfo.x;\r\n        float  capsuleRadius   = capsuleInfo.y;\r\n        float  capsuleScaleTip = capsuleInfo.z;\r\n        float  capsuleOffset   = capsuleInfo.w;\r\n        float3 blobCenter = PositionAndSize[i].xyz;\r\n        float3 toCenter = pos - blobCenter;\r\n        //---------- Capsule SDF ----------\r\n        // Generate a capsule for the tail effect. \r\n        // NOTE: Ground filleting is calculated here; check the h2 variable and radius lerp.\r\n        float dCapsule  = BIG;\r\n        //bool  hasCapsule = (capsuleRadius > 1e-4) &&&& (capsuleHeight > 1e-4);\r\n        bool  hasCapsule = true;\r\n        if (hasCapsule)\r\n        {\r\n            float3 ba = float3(0.0, 0.0, -(capsuleHeight + capsuleOffset));  \r\n            float baLen2 = dot(ba, ba);\r\n            if (baLen2 > 1e-8)\r\n            {\r\n                float  h = saturate(dot(toCenter, ba) / baLen2);\r\n                float3 closest = blobCenter + ba * h;\r\n                float radiusScale = lerp(1.0, capsuleScaleTip, h);\r\n                float h2 = 1-saturate((closest.z - Grounds[i].w)/GFDistance);\r\n                radiusScale = lerp(radiusScale,GFScale,h2*h2);\r\n                float capRadiusLocal = capsuleRadius * radiusScale;\r\n                dCapsule = length(pos - closest) - capRadiusLocal;\r\n                if (dTotalCapsule == BIG)\r\n                {\r\n                    dTotalCapsule = dCapsule;\r\n                }\r\n                else\r\n                {\r\n                    dTotalCapsule = min(dTotalCapsule,dCapsule);\r\n                }\r\n            }\r\n        }\r\n    }\r\n    if(dTotalCapsule!=BIG){\r\n        float dMerge = SmoothUnion(dTotalBlob, dTotalCapsule, smoothK.z);\r\n        dFinal = dMerge;\r\n    }\r\n    ##endif\r\n\r\n    return dFinal;\r\n"
    showCode=true
    inputs [{
      inputName=UnionFunc
      input {
       expression=&self:MaterialExpressionCustom_3
       inputName=None
       mask=0
       maskA=0
       maskB=0
       maskG=0
       maskR=0
       outputIndex=0
      }
     }]
   }
  }
  MaterialExpressionCustom_3 {
   class=MaterialExpressionCustom
   inputs [A {from=MaterialExpressionAdd_13, output=""}]
   properties {
    code="return A;}\r\n\r\nfloat SmoothUnion(float d1, float d2, float k)\r\n{\r\n    float h = saturate(0.5 + 0.5 * (d2 - d1) / k);\r\n    return lerp(d2, d1, h) - k * h * (1.0 - h);"
    showCode=true
    inputs [{
      inputName=A
      input {
       expression=&self:MaterialExpressionAdd_13
       inputName=None
       mask=0
       maskA=0
       maskB=0
       maskG=0
       maskR=0
       outputIndex=0
      }
     }]
   }
  }
  MaterialExpressionCustom_4 {
   class=MaterialExpressionCustom
   inputs [A {from=MaterialExpressionConstant_0, output=""}]
   properties {
    code="return A;}\r\n##define FEATURE_TAIL 1\r\nfloat VoidFunc(float a){\r\nreturn a+1;"
    showCode=true
    inputs [{
      inputName=A
      input {
       expression=&self:MaterialExpressionConstant_0
       inputName=None
       mask=0
       maskA=0
       maskB=0
       maskG=0
       maskR=0
       outputIndex=0
      }
     }]
   }
  }
  MaterialExpressionCustom_5 {
   class=MaterialExpressionCustom
   inputs [None {from="", output=""}]
   properties {code="return 0;}\r\n##define FEATURE_GROUNDFADE 1\r\nfloat VoidFunc2(float a){\r\nreturn a+1;", showCode=true}
  }
  MaterialExpressionCustom_6 {
   class=MaterialExpressionCustom
   inputs [None {from="", output=""}]
   properties {code="return 0;}\r\n##define FEATURE_TWOSPHEREBUBBLE 1\r\nfloat VoidFunc3(float a){\r\nreturn a+1;", showCode=true}
  }
  MaterialExpressionCustom_7 {
   class=MaterialExpressionCustom
   inputs [None {from="", output=""}]
   properties {code="return 0;}\r\n##define FEATURE_TWOSPHEREBUBBLE 0\r\nfloat VoidFunc3(float a){\r\nreturn a+1;", showCode=true}
  }
  MaterialExpressionDistance_1 {
   class=MaterialExpressionDistance
   inputs [
    A {from=MaterialExpressionAdd_0, output=""}
    B {from=MaterialExpressionMaterialFunctionCall_27, output=Result}
   ]
   properties {}
  }
  MaterialExpressionDivide_0 {
   class=MaterialExpressionDivide
   inputs [
    A {from=MaterialExpressionNamedRerouteUsage_32}
    B {from=""}
   ]
   properties {constB=1.97}
  }
  MaterialExpressionDivide_1 {
   class=MaterialExpressionDivide
   inputs [
    A {from=MaterialExpressionComponentMask_3}
    B {from=MaterialExpressionComponentMask_2}
   ]
  }
  MaterialExpressionDivide_10 {
   class=MaterialExpressionDivide
   inputs [
    A {from=MaterialExpressionConvert_0}
    B {from=MaterialExpressionConstant_65}
   ]
  }
  MaterialExpressionDivide_4 {
   class=MaterialExpressionDivide
   inputs [
    A {from=MaterialExpressionDistance_1}
    B {from=MaterialExpressionScalarParameter_19}
   ]
  }
  MaterialExpressionDivide_6 {
   class=MaterialExpressionDivide
   inputs [
    A {from=MaterialExpressionNamedRerouteUsage_14}
    B {from=MaterialExpressionScalarParameter_12}
   ]
  }
  MaterialExpressionDivide_7 {
   class=MaterialExpressionDivide
   inputs [
    A {from=MaterialExpressionMaterialFunctionCall_19, output=Vector}
    B {from=MaterialExpressionConstant4Vector_3, output=B}
   ]
  }
  MaterialExpressionDivide_8 {
   class=MaterialExpressionDivide
   inputs [
    A {from=MaterialExpressionMaterialFunctionCall_19, output=Vector}
    B {from=MaterialExpressionConstant4Vector_3, output=G}
   ]
  }
  MaterialExpressionDivide_9 {
   class=MaterialExpressionDivide
   inputs [
    A {from=MaterialExpressionMaterialFunctionCall_19, output=Vector}
    B {from=MaterialExpressionConstant4Vector_3, output=R}
   ]
  }
  MaterialExpressionDotProduct_0 {
   class=MaterialExpressionDotProduct
   inputs [
    A {from=MaterialExpressionNamedRerouteUsage_29, output=""}
    B {from=MaterialExpressionReroute_3, output=""}
   ]
   properties {}
  }
  MaterialExpressionDotProduct_1 {
   class=MaterialExpressionDotProduct
   inputs [
    A {from=MaterialExpressionCameraVectorWS_0, output=""}
    B {from=MaterialExpressionNamedRerouteUsage_30, output=""}
   ]
   properties {}
  }
  MaterialExpressionDotProduct_2 {
   class=MaterialExpressionDotProduct
   inputs [
    A {from=MaterialExpressionSkyAtmosphereLightDirection_2, output=""}
    B {from=MaterialExpressionMaterialFunctionCall_31, output=Result}
   ]
   properties {}
  }
  MaterialExpressionFresnel_1 {
   class=MaterialExpressionFresnel
   inputs [
    ExponentIn {from=MaterialExpressionConstant4Vector_5, output=R}
    BaseReflectFractionIn {from=MaterialExpressionConstant4Vector_5, output=R}
    Normal {from=MaterialExpressionNamedRerouteUsage_25, output=""}
   ]
   properties {}
  }
  MaterialExpressionFresnel_2 {
   class=MaterialExpressionFresnel
   inputs [
    ExponentIn {from=MaterialExpressionConstant4Vector_6, output=R}
    BaseReflectFractionIn {from=MaterialExpressionConstant4Vector_6, output=R}
    Normal {from=MaterialExpressionNamedRerouteUsage_25, output=""}
   ]
   properties {}
  }
  MaterialExpressionIf_0 {
   class=MaterialExpressionIf
   inputs [
    A {from=MaterialExpressionPixelDepth_0, output=R}
    B {from=MaterialExpressionSceneDepth_0, output=R}
    "A > B" {from=MaterialExpressionConstant_1, output=""}
    "A == B" {from=MaterialExpressionConstant_2, output=""}
    "A < B" {from=MaterialExpressionConstant_2, output=""}
   ]
   properties {}
  }
  MaterialExpressionIf_1 {
   class=MaterialExpressionIf
   inputs [
    A {from=MaterialExpressionMaterialFunctionCall_22, output=T1}
    B {from=MaterialExpressionMaterialFunctionCall_22, output=T1}
    "A > B" {from=MaterialExpressionConstant_9, output=""}
    "A == B" {from=MaterialExpressionConstant_9, output=""}
    "A < B" {from=MaterialExpressionConstant_10, output=""}
   ]
   properties {}
  }
  MaterialExpressionIf_2 {
   class=MaterialExpressionIf
   inputs [
    A {from=MaterialExpressionMaterialFunctionCall_23, output=T1}
    B {from=MaterialExpressionMaterialFunctionCall_23, output=T1}
    "A > B" {from=MaterialExpressionConstant_11, output=""}
    "A == B" {from=MaterialExpressionConstant_11, output=""}
    "A < B" {from=MaterialExpressionConstant_12, output=""}
   ]
   properties {}
  }
  MaterialExpressionIf_3 {
   class=MaterialExpressionIf
   inputs [
    A {from=MaterialExpressionMaterialFunctionCall_24, output=T1}
    B {from=MaterialExpressionMaterialFunctionCall_24, output=T1}
    "A > B" {from=MaterialExpressionConstant_13, output=""}
    "A == B" {from=MaterialExpressionConstant_13, output=""}
    "A < B" {from=MaterialExpressionConstant_14, output=""}
   ]
   properties {}
  }
  MaterialExpressionIf_4 {
   class=MaterialExpressionIf
   inputs [
    A {from=MaterialExpressionMaterialFunctionCall_25, output=T1}
    B {from=MaterialExpressionMaterialFunctionCall_25, output=T1}
    "A > B" {from=MaterialExpressionConstant_15, output=""}
    "A == B" {from=MaterialExpressionConstant_15, output=""}
    "A < B" {from=MaterialExpressionConstant_16, output=""}
   ]
   properties {}
  }
  MaterialExpressionIf_5 {
   class=MaterialExpressionIf
   inputs [
    A {from=MaterialExpressionMaterialFunctionCall_26, output=T1}
    B {from=MaterialExpressionMaterialFunctionCall_26, output=T1}
    "A > B" {from=MaterialExpressionConstant_17, output=""}
    "A == B" {from=MaterialExpressionConstant_17, output=""}
    "A < B" {from=MaterialExpressionConstant_18, output=""}
   ]
   properties {}
  }
  MaterialExpressionLength_0 {class=MaterialExpressionLength, inputs [None {from=MaterialExpressionSubtract_2, output=""}], properties {}}
  MaterialExpressionLength_1 {class=MaterialExpressionLength, inputs [None {from=MaterialExpressionSubtract_5, output=""}], properties {}}
  MaterialExpressionLinearInterpolate_0 {
   class=MaterialExpressionLinearInterpolate
   inputs [
    A {from=""}
    B {from=MaterialExpressionIf_0}
    Alpha {from=MaterialExpressionScalarParameter_13}
   ]
   properties {constA=1}
  }
  MaterialExpressionLinearInterpolate_10 {
   class=MaterialExpressionLinearInterpolate
   inputs [
    A {from=""}
    B {from=MaterialExpressionMultiply_9}
    Alpha {from=MaterialExpressionSaturate_24}
   ]
  }
  MaterialExpressionLinearInterpolate_11 {
   class=MaterialExpressionLinearInterpolate
   inputs [
    A {from=""}
    B {from=MaterialExpressionMultiply_8}
    Alpha {from=MaterialExpressionSaturate_25}
   ]
  }
  MaterialExpressionLinearInterpolate_2 {
   class=MaterialExpressionLinearInterpolate
   inputs [
    A {from=""}
    B {from=MaterialExpressionMultiply_11}
    Alpha {from=MaterialExpressionSaturate_23}
   ]
  }
  MaterialExpressionLinearInterpolate_3 {
   class=MaterialExpressionLinearInterpolate
   inputs [
    A {from=MaterialExpressionSaturate_1}
    B {from=""}
    Alpha {from=MaterialExpressionMaterialFunctionCall_0, output=Result}
   ]
  }
  MaterialExpressionLinearInterpolate_4 {
   class=MaterialExpressionLinearInterpolate
   inputs [
    A {from=""}
    B {from=MaterialExpressionMultiply_12}
    Alpha {from=MaterialExpressionSaturate_3}
   ]
  }
  MaterialExpressionLinearInterpolate_5 {
   class=MaterialExpressionLinearInterpolate
   inputs [
    A {from=MaterialExpressionSkyAtmosphereLightDirection_0}
    B {from=MaterialExpressionConstant3Vector_2}
    Alpha {from=MaterialExpressionConstant_37}
   ]
  }
  MaterialExpressionLinearInterpolate_6 {
   class=MaterialExpressionLinearInterpolate
   inputs [
    A {from=MaterialExpressionLinearInterpolate_5}
    B {from=MaterialExpressionMultiply_32}
    Alpha {from=MaterialExpressionConstant_38}
   ]
  }
  MaterialExpressionLinearInterpolate_7 {
   class=MaterialExpressionLinearInterpolate
   inputs [
    A {from=MaterialExpressionLinearInterpolate_6}
    B {from=MaterialExpressionLinearInterpolate_8}
    Alpha {from=MaterialExpressionOneMinus_5}
   ]
  }
  MaterialExpressionLinearInterpolate_8 {
   class=MaterialExpressionLinearInterpolate
   inputs [
    A {from=MaterialExpressionNormalize_1}
    B {from=MaterialExpressionMaterialFunctionCall_34, output=Result}
    Alpha {from=""}
   ]
   properties {constAlpha=0.8}
  }
  MaterialExpressionLinearInterpolate_9 {
   class=MaterialExpressionLinearInterpolate
   inputs [
    A {from=""}
    B {from=MaterialExpressionMultiply_13}
    Alpha {from=MaterialExpressionSaturate_2}
   ]
  }
  MaterialExpressionMakeMaterialAttributes_0 {
   class=MaterialExpressionMakeMaterialAttributes
   inputs [
    BaseColor {from=MaterialExpressionReroute_6, output=""}
    Metallic {from=MaterialExpressionReroute_5, output=""}
    Specular {from=MaterialExpressionReroute_4, output=""}
    Roughness {from=MaterialExpressionConstant_42, output=""}
    Anisotropy {from="", output=""}
    EmissiveColor {from=MaterialExpressionMaterialFunctionCall_13, output=Result}
    Opacity {from=MaterialExpressionMultiply_10, output=""}
    OpacityMask {from="", output=""}
    Normal {from=MaterialExpressionNamedRerouteUsage_7, output=""}
    Tangent {from="", output=""}
    WorldPositionOffset {from="", output=""}
    SubsurfaceColor {from="", output=""}
    ClearCoat {from="", output=""}
    ClearCoatRoughness {from="", output=""}
    AmbientOcclusion {from="", output=""}
    Refraction {from="", output=""}
    CustomizedUVs_0 {from="", output=""}
    CustomizedUVs_1 {from="", output=""}
    CustomizedUVs_2 {from="", output=""}
    CustomizedUVs_3 {from="", output=""}
    CustomizedUVs_4 {from="", output=""}
    CustomizedUVs_5 {from="", output=""}
    CustomizedUVs_6 {from="", output=""}
    CustomizedUVs_7 {from="", output=""}
    PixelDepthOffset {from="", output=""}
    ShadingModel {from="", output=""}
    Displacement {from="", output=""}
   ]
   properties {}
  }
  MaterialExpressionMaterialFunctionCall_0 {
   class=MaterialExpressionMaterialFunctionCall
   inputs [
    A {from=MaterialExpressionMaterialFunctionCall_1, output=Result}
    B {from=MaterialExpressionActorPositionWS_1}
    Bounds {from=MaterialExpressionConstant3Vector_4}
    Edge Falloff {from=MaterialExpressionConstant_8}
   ]
   properties {materialFunction=/Engine/Functions/Engine_MaterialFunctions01/Gradient/BoxMask-3D}
  }
  MaterialExpressionMaterialFunctionCall_1 {
   class=MaterialExpressionMaterialFunctionCall
   inputs []
   properties {materialFunction=/Game/Effects/Niagara/MasterMaterials/Functions/Common/GetWorldSpacePositionForNearClipPlane}
  }
  MaterialExpressionMaterialFunctionCall_10 {
   class=MaterialExpressionMaterialFunctionCall
   inputs [
    "Epic - PC-High && Epic, XSX, PS5" {from=MaterialExpressionConstant_50}
    "High - XB1, PS4" {from=MaterialExpressionConstant_54}
    "Low - PC Low && Medium" {from=MaterialExpressionConstant_54}
    -------- {from=""}
    Mobile Med - Switch {from=MaterialExpressionConstant_55}
    "Mobile Low - PC_PerfMode, Android" {from=MaterialExpressionConstant_55}
   ]
   properties {materialFunction=/Game/Accessories/FORT_Backpacks/Backpack_QuartzBlob/Material/MF_QuartzBlob_ScalarQualitySwitch}
  }
  MaterialExpressionMaterialFunctionCall_11 {
   class=MaterialExpressionMaterialFunctionCall
   inputs [
    "Xb1,PS4" {from=MaterialExpressionReroute_1}
    "PC-High&&Epic,XSX,PS5" {from=MaterialExpressionMultiply_2}
    "PC-Perf,Android" {from=MaterialExpressionReroute_1}
    Switch {from=MaterialExpressionReroute_1}
    "PC-Low(NotPerf)&&Medium" {from=MaterialExpressionMultiply_2}
   ]
   properties {materialFunction=&PlatformSwitch_1D}
  }
  MaterialExpressionMaterialFunctionCall_13 {
   class=MaterialExpressionMaterialFunctionCall
   inputs [Emissive {from=MaterialExpressionMultiply_37}]
   properties {materialFunction=&MF_FX_EmissiveAdjustments}
  }
  MaterialExpressionMaterialFunctionCall_16 {
   class=MaterialExpressionMaterialFunctionCall
   inputs [
    SceneDepth {from=""}
    Exclude WPO Offsets {from=""}
   ]
   properties {materialFunction=&WorldPositionBehindTranslucency}
  }
  MaterialExpressionMaterialFunctionCall_17 {
   class=MaterialExpressionMaterialFunctionCall
   inputs [
    SceneDepth {from=""}
    Exclude WPO Offsets {from=""}
   ]
   properties {materialFunction=&WorldPositionBehindTranslucency}
  }
  MaterialExpressionMaterialFunctionCall_18 {
   class=MaterialExpressionMaterialFunctionCall
   inputs [
    Camera Relative Position {from=MaterialExpressionAdd_15}
    X Camera to World Vector {from=""}
    Y Camera to World Vector {from=""}
    Z Camera to World Vector {from=""}
    "Tan(FOV/2)*[1,Screen Res Y/X]" {from=""}
   ]
   properties {materialFunction=/Engine/Functions/Engine_MaterialFunctions02/Math/ManualWorldToScreenUVsTransform}
  }
  MaterialExpressionMaterialFunctionCall_19 {
   class=MaterialExpressionMaterialFunctionCall
   inputs [
    Ray Direction {from=MaterialExpressionMaterialFunctionCall_20, output=Result}
    Surface Normal {from=MaterialExpressionNamedRerouteUsage_28}
    Refractive Index Origin {from=MaterialExpressionConstant_40}
    Refractive Index Target {from=MaterialExpressionConstant_41}
   ]
   properties {materialFunction=/Engine/Functions/Engine_MaterialFunctions01/Vectors/Refract}
  }
  MaterialExpressionMaterialFunctionCall_2 {
   class=MaterialExpressionMaterialFunctionCall
   inputs [
    "Epic - PC-High && Epic, XSX, PS5" {from=MaterialExpressionConstant_56}
    "High - XB1, PS4" {from=MaterialExpressionConstant_58}
    "Low - PC Low && Medium" {from=MaterialExpressionConstant_58}
    -------- {from=""}
    Mobile Med - Switch {from=MaterialExpressionConstant_59}
    "Mobile Low - PC_PerfMode, Android" {from=MaterialExpressionConstant_59}
   ]
   properties {materialFunction=&MF_QuartzBlob_ScalarQualitySwitch}
  }
  MaterialExpressionMaterialFunctionCall_20 {
   class=MaterialExpressionMaterialFunctionCall
   inputs []
   properties {materialFunction=&CameraDirectionVector}
  }
  MaterialExpressionMaterialFunctionCall_21 {
   class=MaterialExpressionMaterialFunctionCall
   inputs [
    Camera Relative Position {from=MaterialExpressionAdd_17}
    X Camera to World Vector {from=""}
    Y Camera to World Vector {from=""}
    Z Camera to World Vector {from=""}
    "Tan(FOV/2)*[1,Screen Res Y/X]" {from=""}
   ]
   properties {materialFunction=&ManualWorldToScreenUVsTransform}
  }
  MaterialExpressionMaterialFunctionCall_22 {
   class=MaterialExpressionMaterialFunctionCall
   inputs [
    Ray Origin {from=MaterialExpressionCameraPositionWS_12}
    Ray Direction {from=MaterialExpressionNamedRerouteUsage_16}
    Box Min {from=MaterialExpressionSubtract_14}
    Box Max {from=MaterialExpressionAdd_4}
   ]
   properties {materialFunction=/Engine/Functions/Engine_MaterialFunctions01/Math/BoxIntersection}
  }
  MaterialExpressionMaterialFunctionCall_23 {
   class=MaterialExpressionMaterialFunctionCall
   inputs [
    Ray Origin {from=MaterialExpressionCameraPositionWS_13}
    Ray Direction {from=MaterialExpressionNamedRerouteUsage_17}
    Box Min {from=MaterialExpressionSubtract_13}
    Box Max {from=MaterialExpressionAdd_5}
   ]
   properties {materialFunction=&BoxIntersection}
  }
  MaterialExpressionMaterialFunctionCall_24 {
   class=MaterialExpressionMaterialFunctionCall
   inputs [
    Ray Origin {from=MaterialExpressionCameraPositionWS_14}
    Ray Direction {from=MaterialExpressionNamedRerouteUsage_18}
    Box Min {from=MaterialExpressionSubtract_15}
    Box Max {from=MaterialExpressionAdd_6}
   ]
   properties {materialFunction=&BoxIntersection}
  }
  MaterialExpressionMaterialFunctionCall_25 {
   class=MaterialExpressionMaterialFunctionCall
   inputs [
    Ray Origin {from=MaterialExpressionCameraPositionWS_15}
    Ray Direction {from=MaterialExpressionNamedRerouteUsage_19}
    Box Min {from=MaterialExpressionSubtract_16}
    Box Max {from=MaterialExpressionAdd_7}
   ]
   properties {materialFunction=&BoxIntersection}
  }
  MaterialExpressionMaterialFunctionCall_26 {
   class=MaterialExpressionMaterialFunctionCall
   inputs [
    Ray Origin {from=MaterialExpressionCameraPositionWS_16}
    Ray Direction {from=MaterialExpressionNamedRerouteUsage_20}
    Box Min {from=MaterialExpressionSubtract_17}
    Box Max {from=MaterialExpressionAdd_8}
   ]
   properties {materialFunction=&BoxIntersection}
  }
  MaterialExpressionMaterialFunctionCall_27 {
   class=MaterialExpressionMaterialFunctionCall
   inputs [
    SceneDepth {from=""}
    Exclude WPO Offsets {from=""}
   ]
   properties {materialFunction=&WorldPositionBehindTranslucency}
  }
  MaterialExpressionMaterialFunctionCall_28 {
   class=MaterialExpressionMaterialFunctionCall
   inputs [
    Camera Relative Position {from=MaterialExpressionAdd_16}
    X Camera to World Vector {from=""}
    Y Camera to World Vector {from=""}
    Z Camera to World Vector {from=""}
    "Tan(FOV/2)*[1,Screen Res Y/X]" {from=""}
   ]
   properties {materialFunction=&ManualWorldToScreenUVsTransform}
  }
  MaterialExpressionMaterialFunctionCall_29 {
   class=MaterialExpressionMaterialFunctionCall
   inputs [
    X {from=MaterialExpressionComponentMask_7}
    Y {from=MaterialExpressionComponentMask_8}
    Z {from=MaterialExpressionComponentMask_9}
   ]
   properties {materialFunction=&MakeFloat3}
  }
  MaterialExpressionMaterialFunctionCall_3 {
   class=MaterialExpressionMaterialFunctionCall
   inputs [MaskIn {from=MaterialExpressionMax_3}]
   properties {materialFunction=/Game/Effects/Niagara/MasterMaterials/Functions/Common/TranslucentClip}
  }
  MaterialExpressionMaterialFunctionCall_30 {
   class=MaterialExpressionMaterialFunctionCall
   inputs [
    LightDirctionIn {from=MaterialExpressionReroute_3}
    CameraVectorIn {from=MaterialExpressionCameraVectorWS_0}
    Roughness {from=MaterialExpressionScalarParameter_0}
    NormalIn {from=MaterialExpressionNamedRerouteUsage_29}
   ]
   properties {materialFunction=/Game/Effects/Niagara/MasterMaterials/Functions/Common/FXCommon_GGX}
  }
  MaterialExpressionMaterialFunctionCall_31 {
   class=MaterialExpressionMaterialFunctionCall
   inputs []
   properties {materialFunction=&CameraDirectionVector}
  }
  MaterialExpressionMaterialFunctionCall_32 {
   class=MaterialExpressionMaterialFunctionCall
   inputs [
    Input {from=MaterialExpressionSaturate_17}
    Input Low {from=""}
    Input High {from=""}
    Target Low {from=MaterialExpressionScalarParameter_1}
    Target High {from=""}
   ]
   properties {materialFunction=&RemapValueRange}
  }
  MaterialExpressionMaterialFunctionCall_33 {
   class=MaterialExpressionMaterialFunctionCall
   inputs [
    Plane Normal {from=MaterialExpressionMaterialFunctionCall_34, output=Result}
    Vector {from=MaterialExpressionConstant3Vector_5}
   ]
   properties {materialFunction=/Engine/Functions/Engine_MaterialFunctions02/Math/ProjectVectorOntoPlane}
  }
  MaterialExpressionMaterialFunctionCall_34 {
   class=MaterialExpressionMaterialFunctionCall
   inputs []
   properties {materialFunction=&CameraDirectionVector}
  }
  MaterialExpressionMaterialFunctionCall_35 {
   class=MaterialExpressionMaterialFunctionCall
   inputs [
    CurveIndex {from=MaterialExpressionScalarParameter_5}
    CurveNum {from=MaterialExpressionConstant_3}
    GreyMapIn {from=MaterialExpressionSaturate_14}
   ]
   properties {materialFunction=/Game/Effects/Niagara/MasterMaterials/Functions/Common/FXCommon_CurveAtlasUV}
  }
  MaterialExpressionMaterialFunctionCall_4 {
   class=MaterialExpressionMaterialFunctionCall
   inputs [
    "Epic - PC-High && Epic, XSX, PS5" {from=MaterialExpressionConstant_57}
    "High - XB1, PS4" {from=MaterialExpressionConstant_60}
    "Low - PC Low && Medium" {from=MaterialExpressionConstant_60}
    -------- {from=""}
    Mobile Med - Switch {from=MaterialExpressionConstant_61}
    "Mobile Low - PC_PerfMode, Android" {from=MaterialExpressionConstant_61}
   ]
   properties {materialFunction=&MF_QuartzBlob_ScalarQualitySwitch}
  }
  MaterialExpressionMaterialFunctionCall_5 {
   class=MaterialExpressionMaterialFunctionCall
   inputs [
    "Epic - PC-High && Epic, XSX, PS5" {from=MaterialExpressionConstant_44}
    "High - XB1, PS4" {from=MaterialExpressionConstant_45}
    "Low - PC Low && Medium" {from=MaterialExpressionConstant_45}
    -------- {from=""}
    Mobile Med - Switch {from=MaterialExpressionConstant_46}
    "Mobile Low - PC_PerfMode, Android" {from=MaterialExpressionConstant_46}
   ]
   properties {materialFunction=&MF_QuartzBlob_ScalarQualitySwitch}
  }
  MaterialExpressionMaterialFunctionCall_6 {
   class=MaterialExpressionMaterialFunctionCall
   inputs [
    "Epic - PC-High && Epic, XSX, PS5" {from=MaterialExpressionConstant_47}
    "High - XB1, PS4" {from=MaterialExpressionConstant_48}
    "Low - PC Low && Medium" {from=MaterialExpressionConstant_48}
    -------- {from=""}
    Mobile Med - Switch {from=MaterialExpressionConstant_49}
    "Mobile Low - PC_PerfMode, Android" {from=MaterialExpressionConstant_49}
   ]
   properties {materialFunction=&MF_QuartzBlob_ScalarQualitySwitch}
  }
  MaterialExpressionMaterialFunctionCall_7 {
   class=MaterialExpressionMaterialFunctionCall
   inputs [
    "Xb1,PS4" {from=MaterialExpressionCustom_4}
    "PC-High&&Epic,XSX,PS5" {from=MaterialExpressionCustom_4}
    "PC-Perf,Android" {from=MaterialExpressionCustom_1}
    Switch {from=MaterialExpressionCustom_1}
    "PC-Low(NotPerf)&&Medium" {from=MaterialExpressionCustom_4}
   ]
   properties {materialFunction=&PlatformSwitch_1D}
  }
  MaterialExpressionMaterialFunctionCall_8 {
   class=MaterialExpressionMaterialFunctionCall
   inputs [
    "Xb1,PS4" {from=MaterialExpressionCustom_7}
    "PC-High&&Epic,XSX,PS5" {from=MaterialExpressionCustom_6}
    "PC-Perf,Android" {from=MaterialExpressionCustom_7}
    Switch {from=MaterialExpressionCustom_7}
    "PC-Low(NotPerf)&&Medium" {from=MaterialExpressionCustom_6}
   ]
   properties {materialFunction=&PlatformSwitch_1D}
  }
  MaterialExpressionMaterialFunctionCall_9 {
   class=MaterialExpressionMaterialFunctionCall
   inputs [
    "Epic - PC-High && Epic, XSX, PS5" {from=MaterialExpressionConstant_51}
    "High - XB1, PS4" {from=MaterialExpressionConstant_52}
    "Low - PC Low && Medium" {from=MaterialExpressionConstant_52}
    -------- {from=""}
    Mobile Med - Switch {from=MaterialExpressionConstant_53}
    "Mobile Low - PC_PerfMode, Android" {from=MaterialExpressionConstant_53}
   ]
   properties {materialFunction=&MF_QuartzBlob_ScalarQualitySwitch}
  }
  MaterialExpressionMax_0 {
   class=MaterialExpressionMax
   inputs [
    A {from=MaterialExpressionMultiply_8}
    B {from=MaterialExpressionMultiply_9}
   ]
  }
  MaterialExpressionMax_1 {
   class=MaterialExpressionMax
   inputs [
    A {from=MaterialExpressionMax_0}
    B {from=MaterialExpressionMultiply_11}
   ]
  }
  MaterialExpressionMax_2 {
   class=MaterialExpressionMax
   inputs [
    A {from=MaterialExpressionMax_1}
    B {from=MaterialExpressionMultiply_12}
   ]
  }
  MaterialExpressionMax_3 {
   class=MaterialExpressionMax
   inputs [
    A {from=MaterialExpressionMax_2}
    B {from=MaterialExpressionMultiply_13}
   ]
  }
  MaterialExpressionMax_5 {
   class=MaterialExpressionMax
   inputs [
    A {from=MaterialExpressionVectorParameter_0, output=A}
    B {from=MaterialExpressionNamedRerouteUsage_22}
   ]
  }
  MaterialExpressionMax_6 {
   class=MaterialExpressionMax
   inputs [
    A {from=MaterialExpressionVectorParameter_25, output=A}
    B {from=MaterialExpressionNamedRerouteUsage_22}
   ]
  }
  MaterialExpressionMax_7 {
   class=MaterialExpressionMax
   inputs [
    A {from=MaterialExpressionVectorParameter_26, output=A}
    B {from=MaterialExpressionNamedRerouteUsage_22}
   ]
  }
  MaterialExpressionMax_8 {
   class=MaterialExpressionMax
   inputs [
    A {from=MaterialExpressionVectorParameter_27, output=A}
    B {from=MaterialExpressionNamedRerouteUsage_22}
   ]
  }
  MaterialExpressionMax_9 {
   class=MaterialExpressionMax
   inputs [
    A {from=MaterialExpressionVectorParameter_28, output=A}
    B {from=MaterialExpressionNamedRerouteUsage_22}
   ]
  }
  MaterialExpressionMultiply_0 {
   class=MaterialExpressionMultiply
   inputs [
    A {from=MaterialExpressionViewProperty_0, output=Property}
    B {from=""}
   ]
   properties {constB=2}
  }
  MaterialExpressionMultiply_1 {
   class=MaterialExpressionMultiply
   inputs [
    A {from=MaterialExpressionConstant4Vector_4, output=RGB}
    B {from=MaterialExpressionFresnel_2}
   ]
  }
  MaterialExpressionMultiply_10 {
   class=MaterialExpressionMultiply
   inputs [
    A {from=MaterialExpressionParticleColor_0, output=A}
    B {from=MaterialExpressionMultiply_19}
   ]
  }
  MaterialExpressionMultiply_11 {
   class=MaterialExpressionMultiply
   inputs [
    A {from=MaterialExpressionIf_3}
    B {from=MaterialExpressionSaturate_5}
   ]
  }
  MaterialExpressionMultiply_12 {
   class=MaterialExpressionMultiply
   inputs [
    A {from=MaterialExpressionIf_4}
    B {from=MaterialExpressionSaturate_6}
   ]
  }
  MaterialExpressionMultiply_13 {
   class=MaterialExpressionMultiply
   inputs [
    A {from=MaterialExpressionIf_5}
    B {from=MaterialExpressionSaturate_7}
   ]
  }
  MaterialExpressionMultiply_14 {
   class=MaterialExpressionMultiply
   inputs [
    A {from=MaterialExpressionConvert_3}
    B {from=MaterialExpressionNamedRerouteUsage_15}
   ]
   properties {constB=2}
  }
  MaterialExpressionMultiply_15 {
   class=MaterialExpressionMultiply
   inputs [
    A {from=MaterialExpressionConvert_1}
    B {from=MaterialExpressionNamedRerouteUsage_15}
   ]
   properties {constB=2}
  }
  MaterialExpressionMultiply_16 {
   class=MaterialExpressionMultiply
   inputs [
    A {from=MaterialExpressionConvert_4}
    B {from=MaterialExpressionNamedRerouteUsage_15}
   ]
   properties {constB=2}
  }
  MaterialExpressionMultiply_17 {
   class=MaterialExpressionMultiply
   inputs [
    A {from=MaterialExpressionConvert_5}
    B {from=MaterialExpressionNamedRerouteUsage_15}
   ]
   properties {constB=2}
  }
  MaterialExpressionMultiply_18 {
   class=MaterialExpressionMultiply
   inputs [
    A {from=MaterialExpressionConvert_6}
    B {from=MaterialExpressionNamedRerouteUsage_15}
   ]
   properties {constB=2}
  }
  MaterialExpressionMultiply_19 {
   class=MaterialExpressionMultiply
   inputs [
    A {from=MaterialExpressionMultiply_7}
    B {from=MaterialExpressionLinearInterpolate_0}
   ]
  }
  MaterialExpressionMultiply_2 {
   class=MaterialExpressionMultiply
   inputs [
    A {from=MaterialExpressionConstant4Vector_2, output=RGB}
    B {from=MaterialExpressionVectorParameter_3, output=RGB}
   ]
   properties {constB=0.1}
  }
  MaterialExpressionMultiply_20 {
   class=MaterialExpressionMultiply
   inputs [
    A {from=MaterialExpressionMaterialFunctionCall_29, output=Result}
    B {from=MaterialExpressionConstant4Vector_2, output=RGB}
   ]
  }
  MaterialExpressionMultiply_21 {
   class=MaterialExpressionMultiply
   inputs [
    A {from=MaterialExpressionAdd_19}
    B {from=MaterialExpressionConstant_19}
   ]
  }
  MaterialExpressionMultiply_22 {
   class=MaterialExpressionMultiply
   inputs [
    A {from=MaterialExpressionAdd_20}
    B {from=MaterialExpressionConstant_23}
   ]
  }
  MaterialExpressionMultiply_23 {
   class=MaterialExpressionMultiply
   inputs [
    A {from=MaterialExpressionNamedRerouteUsage_5}
    B {from=MaterialExpressionNamedRerouteUsage_6}
   ]
  }
  MaterialExpressionMultiply_24 {
   class=MaterialExpressionMultiply
   inputs [
    A {from=MaterialExpressionMultiply_4}
    B {from=MaterialExpressionSaturate_12}
   ]
  }
  MaterialExpressionMultiply_25 {
   class=MaterialExpressionMultiply
   inputs [
    A {from=MaterialExpressionMultiply_34}
    B {from=MaterialExpressionConstant_21}
   ]
  }
  MaterialExpressionMultiply_26 {
   class=MaterialExpressionMultiply
   inputs [
    A {from=MaterialExpressionSaturate_13}
    B {from=MaterialExpressionConstant_25}
   ]
  }
  MaterialExpressionMultiply_27 {
   class=MaterialExpressionMultiply
   inputs [
    A {from=MaterialExpressionDotProduct_0}
    B {from=""}
   ]
   properties {constB=-1}
  }
  MaterialExpressionMultiply_28 {
   class=MaterialExpressionMultiply
   inputs [
    A {from=MaterialExpressionMaterialFunctionCall_30, output=Result}
    B {from=MaterialExpressionVectorParameter_2, output=RGB}
   ]
  }
  MaterialExpressionMultiply_29 {
   class=MaterialExpressionMultiply
   inputs [
    A {from=MaterialExpressionAdd_23}
    B {from=MaterialExpressionConstant_27}
   ]
  }
  MaterialExpressionMultiply_3 {
   class=MaterialExpressionMultiply
   inputs [
    A {from=MaterialExpressionDivide_1}
    B {from=MaterialExpressionConstant_66}
   ]
  }
  MaterialExpressionMultiply_30 {
   class=MaterialExpressionMultiply
   inputs [
    A {from=MaterialExpressionMaterialFunctionCall_32, output=Result}
    B {from=MaterialExpressionSkyLightEnvMapSample_0}
   ]
  }
  MaterialExpressionMultiply_31 {
   class=MaterialExpressionMultiply
   inputs [
    A {from=MaterialExpressionVectorParameter_1, output=RGB}
    B {from=MaterialExpressionMultiply_30}
   ]
  }
  MaterialExpressionMultiply_32 {
   class=MaterialExpressionMultiply
   inputs [
    A {from=MaterialExpressionCameraVectorWS_1}
    B {from=""}
   ]
   properties {constB=-1}
  }
  MaterialExpressionMultiply_33 {
   class=MaterialExpressionMultiply
   inputs [
    A {from=MaterialExpressionAdd_25}
    B {from=MaterialExpressionConstant_32}
   ]
  }
  MaterialExpressionMultiply_34 {
   class=MaterialExpressionMultiply
   inputs [
    A {from=MaterialExpressionOneMinus_4}
    B {from=MaterialExpressionSaturate_10}
   ]
  }
  MaterialExpressionMultiply_35 {
   class=MaterialExpressionMultiply
   inputs [
    A {from=MaterialExpressionAdd_26}
    B {from=MaterialExpressionConstant_35}
   ]
  }
  MaterialExpressionMultiply_36 {
   class=MaterialExpressionMultiply
   inputs [
    A {from=MaterialExpressionAdd_27}
    B {from=MaterialExpressionConstant_63}
   ]
  }
  MaterialExpressionMultiply_37 {
   class=MaterialExpressionMultiply
   inputs [
    A {from=MaterialExpressionParticleColor_0, output=RGB}
    B {from=MaterialExpressionDivide_0}
   ]
  }
  MaterialExpressionMultiply_4 {
   class=MaterialExpressionMultiply
   inputs [
    A {from=MaterialExpressionMultiply_23}
    B {from=MaterialExpressionNamedRerouteUsage_33}
   ]
   properties {constA=1}
  }
  MaterialExpressionMultiply_5 {
   class=MaterialExpressionMultiply
   inputs [
    A {from=MaterialExpressionNamedRerouteUsage_4}
    B {from=MaterialExpressionNamedRerouteUsage_8}
   ]
  }
  MaterialExpressionMultiply_6 {
   class=MaterialExpressionMultiply
   inputs [
    A {from=MaterialExpressionDivide_4}
    B {from=MaterialExpressionDivide_4}
   ]
  }
  MaterialExpressionMultiply_7 {
   class=MaterialExpressionMultiply
   inputs [
    A {from=MaterialExpressionPower_0}
    B {from=MaterialExpressionMultiply_24}
   ]
  }
  MaterialExpressionMultiply_8 {
   class=MaterialExpressionMultiply
   inputs [
    A {from=MaterialExpressionIf_1}
    B {from=MaterialExpressionSaturate_9}
   ]
  }
  MaterialExpressionMultiply_9 {
   class=MaterialExpressionMultiply
   inputs [
    A {from=MaterialExpressionIf_2}
    B {from=MaterialExpressionSaturate_4}
   ]
  }
  MaterialExpressionNamedRerouteDeclaration_0 {
   class=MaterialExpressionNamedRerouteDeclaration
   inputs [None {from=MaterialExpressionMultiply_3}]
   properties {name=PixelScale, nodeColor=<rgba 0.7578 1 0 1>}
  }
  MaterialExpressionNamedRerouteDeclaration_1 {
   class=MaterialExpressionNamedRerouteDeclaration
   inputs [None {from=MaterialExpressionComponentMask_1}]
   properties {name=OpacityMask, nodeColor=<rgba 1 0.9141 0 1>}
  }
  MaterialExpressionNamedRerouteDeclaration_10 {
   class=MaterialExpressionNamedRerouteDeclaration
   inputs [None {from=MaterialExpressionMaterialFunctionCall_4, output=Result}]
   properties {name=EdgeWidth, nodeColor=<rgba 0.9688 0 1 1>}
  }
  MaterialExpressionNamedRerouteDeclaration_11 {
   class=MaterialExpressionNamedRerouteDeclaration
   inputs [None {from=MaterialExpressionMaterialFunctionCall_5, output=Result}]
   properties {name=RenderingBoundScale, nodeColor=<rgba 0.9688 0 1 1>}
  }
  MaterialExpressionNamedRerouteDeclaration_12 {
   class=MaterialExpressionNamedRerouteDeclaration
   inputs [None {from=MaterialExpressionMaterialFunctionCall_6, output=Result}]
   properties {name=RenderingBoundMinSize, nodeColor=<rgba 0.9688 0 1 1>}
  }
  MaterialExpressionNamedRerouteDeclaration_13 {
   class=MaterialExpressionNamedRerouteDeclaration
   inputs [None {from=MaterialExpressionMaterialFunctionCall_3, output=ToNormalOrEmissive}]
   properties {name=RenderAreaClip, nodeColor=<rgba 0 0.3437 1 1>}
  }
  MaterialExpressionNamedRerouteDeclaration_14 {
   class=MaterialExpressionNamedRerouteDeclaration
   inputs [None {from=MaterialExpressionAdd_0}]
   properties {name=RNSurfaceWSPos, nodeColor=<rgba 1 0 0.1641 1>}
  }
  MaterialExpressionNamedRerouteDeclaration_15 {
   class=MaterialExpressionNamedRerouteDeclaration
   inputs [None {from=MaterialExpressionMaterialFunctionCall_10, output=Result}]
   properties {name=Threshold, nodeColor=<rgba 1 0.1172 0 1>}
  }
  MaterialExpressionNamedRerouteDeclaration_16 {
   class=MaterialExpressionNamedRerouteDeclaration
   inputs [None {from=MaterialExpressionMaterialFunctionCall_9, output=Result}]
   properties {name=MaxSteps, nodeColor=<rgba 1 0.3281 0 1>}
  }
  MaterialExpressionNamedRerouteDeclaration_17 {
   class=MaterialExpressionNamedRerouteDeclaration
   inputs [None {from=MaterialExpressionCustom_0, output=distToGround}]
   properties {name=DistToGround, nodeColor=<rgba 0.1484 0 1 1>}
  }
  MaterialExpressionNamedRerouteDeclaration_18 {
   class=MaterialExpressionNamedRerouteDeclaration
   inputs [None {from=MaterialExpressionAdd_24}]
   properties {name=RenderStyleOutput, nodeColor=<rgba 0.1953 0 1 1>}
  }
  MaterialExpressionNamedRerouteDeclaration_19 {
   class=MaterialExpressionNamedRerouteDeclaration
   inputs [None {from=MaterialExpressionSaturate_22}]
   properties {name=HeightMask, nodeColor=<rgba 1 0 0.6797 1>}
  }
  MaterialExpressionNamedRerouteDeclaration_2 {
   class=MaterialExpressionNamedRerouteDeclaration
   inputs [None {from=MaterialExpressionComponentMask_0}]
   properties {name=Normal, nodeColor=<rgba 0.5938 1 0 1>}
  }
  MaterialExpressionNamedRerouteDeclaration_3 {
   class=MaterialExpressionNamedRerouteDeclaration
   inputs [None {from=MaterialExpressionLength_0}]
   properties {name=ObjectDist, nodeColor=<rgba 0.9219 0 1 1>}
  }
  MaterialExpressionNamedRerouteDeclaration_4 {
   class=MaterialExpressionNamedRerouteDeclaration
   inputs [None {from=MaterialExpressionLength_1}]
   properties {name=AccurateSceneDepth, nodeColor=<rgba 0 1 0.6016 1>}
  }
  MaterialExpressionNamedRerouteDeclaration_5 {
   class=MaterialExpressionNamedRerouteDeclaration
   inputs [None {from=MaterialExpressionNormalize_0}]
   properties {name=RMDirection, nodeColor=<rgba 1 0.2812 0 1>}
  }
  MaterialExpressionNamedRerouteDeclaration_6 {
   class=MaterialExpressionNamedRerouteDeclaration
   inputs [None {from=MaterialExpressionLinearInterpolate_3}]
   properties {name=CameraInsideMask, nodeColor=<rgba 0 0.0391 1 1>}
  }
  MaterialExpressionNamedRerouteDeclaration_7 {
   class=MaterialExpressionNamedRerouteDeclaration
   inputs [None {from=MaterialExpressionCustom_0, output=Dist}]
   properties {name=RMDist, nodeColor=<rgba 0.3594 1 0 1>}
  }
  MaterialExpressionNamedRerouteDeclaration_9 {
   class=MaterialExpressionNamedRerouteDeclaration
   inputs [None {from=MaterialExpressionMaterialFunctionCall_2, output=Result}]
   properties {name=Epsilon, nodeColor=<rgba 0.7109 1 0 1>}
  }
  MaterialExpressionNamedRerouteUsage_0 {class=MaterialExpressionNamedRerouteUsage}
  MaterialExpressionNamedRerouteUsage_1 {class=MaterialExpressionNamedRerouteUsage}
  MaterialExpressionNamedRerouteUsage_10 {class=MaterialExpressionNamedRerouteUsage}
  MaterialExpressionNamedRerouteUsage_11 {class=MaterialExpressionNamedRerouteUsage}
  MaterialExpressionNamedRerouteUsage_12 {class=MaterialExpressionNamedRerouteUsage}
  MaterialExpressionNamedRerouteUsage_13 {class=MaterialExpressionNamedRerouteUsage}
  MaterialExpressionNamedRerouteUsage_14 {class=MaterialExpressionNamedRerouteUsage}
  MaterialExpressionNamedRerouteUsage_15 {class=MaterialExpressionNamedRerouteUsage}
  MaterialExpressionNamedRerouteUsage_16 {class=MaterialExpressionNamedRerouteUsage}
  MaterialExpressionNamedRerouteUsage_17 {class=MaterialExpressionNamedRerouteUsage}
  MaterialExpressionNamedRerouteUsage_18 {class=MaterialExpressionNamedRerouteUsage}
  MaterialExpressionNamedRerouteUsage_19 {class=MaterialExpressionNamedRerouteUsage}
  MaterialExpressionNamedRerouteUsage_2 {class=MaterialExpressionNamedRerouteUsage}
  MaterialExpressionNamedRerouteUsage_20 {class=MaterialExpressionNamedRerouteUsage}
  MaterialExpressionNamedRerouteUsage_22 {class=MaterialExpressionNamedRerouteUsage}
  MaterialExpressionNamedRerouteUsage_25 {class=MaterialExpressionNamedRerouteUsage}
  MaterialExpressionNamedRerouteUsage_26 {class=MaterialExpressionNamedRerouteUsage}
  MaterialExpressionNamedRerouteUsage_27 {class=MaterialExpressionNamedRerouteUsage}
  MaterialExpressionNamedRerouteUsage_28 {class=MaterialExpressionNamedRerouteUsage}
  MaterialExpressionNamedRerouteUsage_29 {class=MaterialExpressionNamedRerouteUsage}
  MaterialExpressionNamedRerouteUsage_3 {class=MaterialExpressionNamedRerouteUsage}
  MaterialExpressionNamedRerouteUsage_30 {class=MaterialExpressionNamedRerouteUsage}
  MaterialExpressionNamedRerouteUsage_31 {class=MaterialExpressionNamedRerouteUsage}
  MaterialExpressionNamedRerouteUsage_32 {class=MaterialExpressionNamedRerouteUsage}
  MaterialExpressionNamedRerouteUsage_33 {class=MaterialExpressionNamedRerouteUsage}
  MaterialExpressionNamedRerouteUsage_4 {class=MaterialExpressionNamedRerouteUsage}
  MaterialExpressionNamedRerouteUsage_5 {class=MaterialExpressionNamedRerouteUsage}
  MaterialExpressionNamedRerouteUsage_6 {class=MaterialExpressionNamedRerouteUsage}
  MaterialExpressionNamedRerouteUsage_7 {class=MaterialExpressionNamedRerouteUsage}
  MaterialExpressionNamedRerouteUsage_8 {class=MaterialExpressionNamedRerouteUsage}
  MaterialExpressionNormalize_0 {class=MaterialExpressionNormalize, inputs [VectorInput {from=MaterialExpressionSubtract_0, output=""}]}
  MaterialExpressionNormalize_1 {class=MaterialExpressionNormalize, inputs [VectorInput {from=MaterialExpressionMaterialFunctionCall_33, output=Result}]}
  MaterialExpressionOneMinus_2 {class=MaterialExpressionOneMinus, inputs [None {from=MaterialExpressionAbs_1}]}
  MaterialExpressionOneMinus_4 {class=MaterialExpressionOneMinus, inputs [None {from=MaterialExpressionSaturate_19}]}
  MaterialExpressionOneMinus_5 {class=MaterialExpressionOneMinus, inputs [None {from=MaterialExpressionSaturate_21}]}
  MaterialExpressionParticleColor_0 {class=MaterialExpressionParticleColor, inputs [], properties {}}
  MaterialExpressionPixelDepth_0 {class=MaterialExpressionPixelDepth, inputs [], properties {}}
  MaterialExpressionPower_0 {
   class=MaterialExpressionPower
   inputs [
    Base {from=MaterialExpressionSaturate_0}
    Exp {from=""}
   ]
  }
  MaterialExpressionPower_1 {
   class=MaterialExpressionPower
   inputs [
    Base {from=MaterialExpressionSaturate_8}
    Exp {from=MaterialExpressionConstant_20}
   ]
  }
  MaterialExpressionPower_2 {
   class=MaterialExpressionPower
   inputs [
    Base {from=MaterialExpressionSaturate_11}
    Exp {from=MaterialExpressionConstant_24}
   ]
  }
  MaterialExpressionPower_3 {
   class=MaterialExpressionPower
   inputs [
    Base {from=MaterialExpressionSaturate_16}
    Exp {from=MaterialExpressionConstant_28}
   ]
  }
  MaterialExpressionPower_4 {
   class=MaterialExpressionPower
   inputs [
    Base {from=MaterialExpressionSaturate_18}
    Exp {from=MaterialExpressionConstant_33}
   ]
  }
  MaterialExpressionPower_5 {
   class=MaterialExpressionPower
   inputs [
    Base {from=MaterialExpressionSaturate_20}
    Exp {from=MaterialExpressionConstant_36}
   ]
  }
  MaterialExpressionPower_7 {
   class=MaterialExpressionPower
   inputs [
    Base {from=MaterialExpressionSaturate_15}
    Exp {from=MaterialExpressionConstant_64}
   ]
  }
  MaterialExpressionReflectionVectorWS_0 {
   class=MaterialExpressionReflectionVectorWS
   inputs [CustomWorldNormal {from=MaterialExpressionNamedRerouteUsage_31, output=""}]
   properties {}
  }
  MaterialExpressionReroute_0 {class=MaterialExpressionReroute, inputs [None {from=MaterialExpressionMultiply_20}]}
  MaterialExpressionReroute_1 {class=MaterialExpressionReroute, inputs [None {from=MaterialExpressionMultiply_2}]}
  MaterialExpressionReroute_3 {class=MaterialExpressionReroute, inputs [None {from=MaterialExpressionLinearInterpolate_7}]}
  MaterialExpressionReroute_4 {class=MaterialExpressionReroute, inputs [None {from=MaterialExpressionFresnel_1}]}
  MaterialExpressionReroute_5 {class=MaterialExpressionReroute, inputs [None {from=MaterialExpressionConstant_43}]}
  MaterialExpressionReroute_6 {class=MaterialExpressionReroute, inputs [None {from=MaterialExpressionMultiply_1}]}
  MaterialExpressionSaturate_0 {class=MaterialExpressionSaturate, inputs [None {from=MaterialExpressionDivide_6}]}
  MaterialExpressionSaturate_1 {class=MaterialExpressionSaturate, inputs [None {from=MaterialExpressionTwoSidedSign_1}]}
  MaterialExpressionSaturate_10 {class=MaterialExpressionSaturate, inputs [None {from=MaterialExpressionPower_1}]}
  MaterialExpressionSaturate_11 {class=MaterialExpressionSaturate, inputs [None {from=MaterialExpressionMultiply_22}]}
  MaterialExpressionSaturate_12 {class=MaterialExpressionSaturate, inputs [None {from=MaterialExpressionDivide_4}]}
  MaterialExpressionSaturate_13 {class=MaterialExpressionSaturate, inputs [None {from=MaterialExpressionPower_2}]}
  MaterialExpressionSaturate_14 {class=MaterialExpressionSaturate, inputs [None {from=MaterialExpressionAdd_21}]}
  MaterialExpressionSaturate_15 {class=MaterialExpressionSaturate, inputs [None {from=MaterialExpressionMultiply_36}]}
  MaterialExpressionSaturate_16 {class=MaterialExpressionSaturate, inputs [None {from=MaterialExpressionMultiply_29}]}
  MaterialExpressionSaturate_17 {class=MaterialExpressionSaturate, inputs [None {from=MaterialExpressionPower_3}]}
  MaterialExpressionSaturate_18 {class=MaterialExpressionSaturate, inputs [None {from=MaterialExpressionMultiply_33}]}
  MaterialExpressionSaturate_19 {class=MaterialExpressionSaturate, inputs [None {from=MaterialExpressionPower_4}]}
  MaterialExpressionSaturate_2 {class=MaterialExpressionSaturate, inputs [None {from=MaterialExpressionCeil_9}]}
  MaterialExpressionSaturate_20 {class=MaterialExpressionSaturate, inputs [None {from=MaterialExpressionMultiply_35}]}
  MaterialExpressionSaturate_21 {class=MaterialExpressionSaturate, inputs [None {from=MaterialExpressionPower_5}]}
  MaterialExpressionSaturate_22 {class=MaterialExpressionSaturate, inputs [None {from=MaterialExpressionPower_7}]}
  MaterialExpressionSaturate_23 {class=MaterialExpressionSaturate, inputs [None {from=MaterialExpressionCeil_7}]}
  MaterialExpressionSaturate_24 {class=MaterialExpressionSaturate, inputs [None {from=MaterialExpressionCeil_6}]}
  MaterialExpressionSaturate_25 {class=MaterialExpressionSaturate, inputs [None {from=MaterialExpressionCeil_0}]}
  MaterialExpressionSaturate_3 {class=MaterialExpressionSaturate, inputs [None {from=MaterialExpressionCeil_8}]}
  MaterialExpressionSaturate_4 {class=MaterialExpressionSaturate, inputs [None {from=MaterialExpressionCeil_1}]}
  MaterialExpressionSaturate_5 {class=MaterialExpressionSaturate, inputs [None {from=MaterialExpressionCeil_2}]}
  MaterialExpressionSaturate_6 {class=MaterialExpressionSaturate, inputs [None {from=MaterialExpressionCeil_3}]}
  MaterialExpressionSaturate_7 {class=MaterialExpressionSaturate, inputs [None {from=MaterialExpressionCeil_4}]}
  MaterialExpressionSaturate_8 {class=MaterialExpressionSaturate, inputs [None {from=MaterialExpressionMultiply_21}]}
  MaterialExpressionSaturate_9 {class=MaterialExpressionSaturate, inputs [None {from=MaterialExpressionCeil_5}]}
  MaterialExpressionScalarParameter_0 {class=MaterialExpressionScalarParameter, properties {defaultValue=0.02, group="######ColorTweak######", parameterName=HighLightPointSize}}
  MaterialExpressionScalarParameter_1 {class=MaterialExpressionScalarParameter, properties {defaultValue=0.02, group="######ColorTweak######", parameterName=EnvReflectCenterStrength}}
  MaterialExpressionScalarParameter_12 {class=MaterialExpressionScalarParameter, properties {defaultValue=10, parameterName=DistToGroundFade_Distance}}
  MaterialExpressionScalarParameter_13 {class=MaterialExpressionScalarParameter, properties {parameterName=EnableDepthTest}}
  MaterialExpressionScalarParameter_19 {class=MaterialExpressionScalarParameter, properties {defaultValue=100, parameterName=DepthFade}}
  MaterialExpressionScalarParameter_5 {class=MaterialExpressionScalarParameter, properties {group="######ColorTweak######", parameterName=ColorCurveIndex}}
  MaterialExpressionSceneColor_3 {class=MaterialExpressionSceneColor, inputs [UVs {from=MaterialExpressionMaterialFunctionCall_18, output=Screen UVs}], properties {}}
  MaterialExpressionSceneColor_4 {class=MaterialExpressionSceneColor, inputs [UVs {from=MaterialExpressionMaterialFunctionCall_28, output=Screen UVs}], properties {}}
  MaterialExpressionSceneColor_5 {class=MaterialExpressionSceneColor, inputs [UVs {from=MaterialExpressionMaterialFunctionCall_21, output=Screen UVs}], properties {}}
  MaterialExpressionSceneDepth_0 {class=MaterialExpressionSceneDepth, inputs [UVs {from="", output=""}], properties {}}
  MaterialExpressionSkyAtmosphereLightDirection_0 {class=MaterialExpressionSkyAtmosphereLightDirection, inputs [], properties {}}
  MaterialExpressionSkyAtmosphereLightDirection_2 {class=MaterialExpressionSkyAtmosphereLightDirection, inputs [], properties {}}
  MaterialExpressionSkyLightEnvMapSample_0 {
   class=MaterialExpressionSkyLightEnvMapSample
   inputs [
    Direction {from=MaterialExpressionReflectionVectorWS_0, output=""}
    Roughness {from=MaterialExpressionConstant_30, output=""}
   ]
   properties {}
  }
  MaterialExpressionSubtract_0 {
   class=MaterialExpressionSubtract
   inputs [
    A {from=MaterialExpressionMaterialFunctionCall_16, output=Result}
    B {from=MaterialExpressionCameraPositionWS_1}
   ]
  }
  MaterialExpressionSubtract_1 {
   class=MaterialExpressionSubtract
   inputs [
    A {from=MaterialExpressionActorPositionWS_0}
    B {from=MaterialExpressionVectorParameter_19, output=RGB}
   ]
  }
  MaterialExpressionSubtract_10 {
   class=MaterialExpressionSubtract
   inputs [
    A {from=MaterialExpressionVectorParameter_26, output=RGB}
    B {from=MaterialExpressionMultiply_16}
   ]
  }
  MaterialExpressionSubtract_11 {
   class=MaterialExpressionSubtract
   inputs [
    A {from=MaterialExpressionVectorParameter_27, output=RGB}
    B {from=MaterialExpressionMultiply_17}
   ]
  }
  MaterialExpressionSubtract_12 {
   class=MaterialExpressionSubtract
   inputs [
    A {from=MaterialExpressionVectorParameter_28, output=RGB}
    B {from=MaterialExpressionMultiply_18}
   ]
  }
  MaterialExpressionSubtract_13 {
   class=MaterialExpressionSubtract
   inputs [
    A {from=MaterialExpressionSubtract_9}
    B {from=MaterialExpressionConstant3Vector_3}
   ]
  }
  MaterialExpressionSubtract_14 {
   class=MaterialExpressionSubtract
   inputs [
    A {from=MaterialExpressionSubtract_8}
    B {from=MaterialExpressionConstant3Vector_3}
   ]
  }
  MaterialExpressionSubtract_15 {
   class=MaterialExpressionSubtract
   inputs [
    A {from=MaterialExpressionSubtract_10}
    B {from=MaterialExpressionConstant3Vector_3}
   ]
  }
  MaterialExpressionSubtract_16 {
   class=MaterialExpressionSubtract
   inputs [
    A {from=MaterialExpressionSubtract_11}
    B {from=MaterialExpressionConstant3Vector_3}
   ]
  }
  MaterialExpressionSubtract_17 {
   class=MaterialExpressionSubtract
   inputs [
    A {from=MaterialExpressionSubtract_12}
    B {from=MaterialExpressionConstant3Vector_3}
   ]
  }
  MaterialExpressionSubtract_2 {
   class=MaterialExpressionSubtract
   inputs [
    A {from=MaterialExpressionActorPositionWS_2}
    B {from=MaterialExpressionCameraPositionWS_2}
   ]
  }
  MaterialExpressionSubtract_3 {
   class=MaterialExpressionSubtract
   inputs [
    A {from=MaterialExpressionAdd_0}
    B {from=MaterialExpressionActorPositionWS_3}
   ]
  }
  MaterialExpressionSubtract_5 {
   class=MaterialExpressionSubtract
   inputs [
    A {from=MaterialExpressionMaterialFunctionCall_17, output=Result}
    B {from=MaterialExpressionCameraPositionWS_9}
   ]
  }
  MaterialExpressionSubtract_6 {
   class=MaterialExpressionSubtract
   inputs [
    A {from=MaterialExpressionNamedRerouteUsage_27}
    B {from=MaterialExpressionCameraPositionWS_4}
   ]
  }
  MaterialExpressionSubtract_8 {
   class=MaterialExpressionSubtract
   inputs [
    A {from=MaterialExpressionVectorParameter_0, output=RGB}
    B {from=MaterialExpressionMultiply_15}
   ]
  }
  MaterialExpressionSubtract_9 {
   class=MaterialExpressionSubtract
   inputs [
    A {from=MaterialExpressionVectorParameter_25, output=RGB}
    B {from=MaterialExpressionMultiply_14}
   ]
  }
  MaterialExpressionTextureSampleParameter2D_0 {
   class=MaterialExpressionTextureSampleParameter2D
   inputs [
    UVs {from=MaterialExpressionMaterialFunctionCall_35, output=UV}
    Apply View MipBias {from="", output=""}
   ]
   properties {
    group="######ColorTweak######"
    parameterName=CurveTexture
    samplerType=SAMPLERTYPE_LinearColor
    texture=&/CLA_LavaBlob
   }
  }
  MaterialExpressionTwoSidedSign_1 {class=MaterialExpressionTwoSidedSign, inputs [], properties {}}
  MaterialExpressionVectorParameter_0 {class=MaterialExpressionVectorParameter, properties {parameterName=PosAndSize0}}
  MaterialExpressionVectorParameter_1 {
   class=MaterialExpressionVectorParameter
   properties {group="######ColorTweak######", parameterName=EnvReflectColor, defaultValue=<rgba 0.2 0.2 0.2 1>}
  }
  MaterialExpressionVectorParameter_10 {class=MaterialExpressionVectorParameter, properties {parameterName=Capsule1}}
  MaterialExpressionVectorParameter_11 {class=MaterialExpressionVectorParameter, properties {parameterName=Capsule2}}
  MaterialExpressionVectorParameter_12 {class=MaterialExpressionVectorParameter, properties {parameterName=Capsule3}}
  MaterialExpressionVectorParameter_13 {class=MaterialExpressionVectorParameter, properties {parameterName=Capsule4}}
  MaterialExpressionVectorParameter_14 {class=MaterialExpressionVectorParameter, properties {parameterName=SecondBubble0}}
  MaterialExpressionVectorParameter_15 {class=MaterialExpressionVectorParameter, properties {parameterName=SecondBubble1}}
  MaterialExpressionVectorParameter_16 {class=MaterialExpressionVectorParameter, properties {parameterName=SecondBubble2}}
  MaterialExpressionVectorParameter_17 {class=MaterialExpressionVectorParameter, properties {parameterName=SecondBubble3}}
  MaterialExpressionVectorParameter_18 {class=MaterialExpressionVectorParameter, properties {parameterName=SecondBubble4}}
  MaterialExpressionVectorParameter_19 {class=MaterialExpressionVectorParameter, properties {parameterName=BoxCenter, defaultValue=<rgba 0 0 100 1>}}
  MaterialExpressionVectorParameter_2 {
   class=MaterialExpressionVectorParameter
   properties {group="######ColorTweak######", parameterName=HighLightColor, defaultValue=<rgba 100 91 58 1>}
  }
  MaterialExpressionVectorParameter_20 {class=MaterialExpressionVectorParameter, properties {parameterName=BoxSize, defaultValue=<rgba 256 256 2 1>}}
  MaterialExpressionVectorParameter_21 {class=MaterialExpressionVectorParameter, properties {parameterName=smoothK, defaultValue=<rgba 10 5 10 30>}}
  MaterialExpressionVectorParameter_25 {class=MaterialExpressionVectorParameter, properties {parameterName=PosAndSize1}}
  MaterialExpressionVectorParameter_26 {class=MaterialExpressionVectorParameter, properties {parameterName=PosAndSize2}}
  MaterialExpressionVectorParameter_27 {class=MaterialExpressionVectorParameter, properties {parameterName=PosAndSize3}}
  MaterialExpressionVectorParameter_28 {class=MaterialExpressionVectorParameter, properties {parameterName=PosAndSize4}}
  MaterialExpressionVectorParameter_29 {class=MaterialExpressionVectorParameter, properties {parameterName=Ground0, defaultValue=<rgba 0 0 1 510>}}
  MaterialExpressionVectorParameter_3 {
   class=MaterialExpressionVectorParameter
   properties {group="######ColorTweak######", parameterName="OverLayColor(SimTheSkyLight)", defaultValue=<rgba 0.09 0.1324 0.3 1>}
  }
  MaterialExpressionVectorParameter_30 {class=MaterialExpressionVectorParameter, properties {parameterName=Ground1, defaultValue=<rgba 0 0 1 510>}}
  MaterialExpressionVectorParameter_31 {class=MaterialExpressionVectorParameter, properties {parameterName=Ground2, defaultValue=<rgba 0 0 1 510>}}
  MaterialExpressionVectorParameter_32 {class=MaterialExpressionVectorParameter, properties {parameterName=Ground3, defaultValue=<rgba 0 0 1 510>}}
  MaterialExpressionVectorParameter_33 {class=MaterialExpressionVectorParameter, properties {parameterName=Ground4, defaultValue=<rgba 0 0 1 510>}}
  MaterialExpressionVectorParameter_4 {class=MaterialExpressionVectorParameter, properties {parameterName=PosAndSize0}}
  MaterialExpressionVectorParameter_5 {class=MaterialExpressionVectorParameter, properties {parameterName=PosAndSize1}}
  MaterialExpressionVectorParameter_6 {class=MaterialExpressionVectorParameter, properties {parameterName=PosAndSize2}}
  MaterialExpressionVectorParameter_7 {class=MaterialExpressionVectorParameter, properties {parameterName=PosAndSize3}}
  MaterialExpressionVectorParameter_8 {class=MaterialExpressionVectorParameter, properties {parameterName=PosAndSize4}}
  MaterialExpressionVectorParameter_9 {class=MaterialExpressionVectorParameter, properties {parameterName=Capsule0}}
  MaterialExpressionViewProperty_0 {class=MaterialExpressionViewProperty, inputs [], properties {property=MEVP_TanHalfFieldOfView}}
  MaterialExpressionViewSize_1 {class=MaterialExpressionViewSize, inputs [], properties {}}
 }
 properties {
  bAutomaticallySetUsageInEditor=false
  bDisableDepthTest=true
  bEnableResponsiveAA=true
  bTangentSpaceNormal=false
  bUseMaterialAttributes=true
  bUsedWithNiagaraMeshParticles=true
  blendMode=BLEND_Translucent
  shadingModel=MSM_Unlit
  twoSided=true
  usedShadingModels=Unlit
 }
}
